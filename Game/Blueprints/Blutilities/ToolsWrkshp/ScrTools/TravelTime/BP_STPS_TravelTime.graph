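# BP_STPS_TravelTime is an editor utility Blueprint holding the settings for a Travel Time tool, apparently a spline-based measurer of distance and travel time.
# Its variables cover movement speed presets and multiplier, units, grid snapping, surface following, segment and text styling, and panel open states.
# Clear Current Segments and Clear Other Splines only broadcast the CleanCurSegmentsClicked and CleanAllClicked dispatchers, leaving cleanup to listeners.
# The event graph is empty.
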EditorUtilityBlueprint /Game/Blueprints/Blutilities/ToolsWrkshp/ScrTools/TravelTime/BP_STPS_TravelTime
graphs [
 CleanAllClicked {nodes [CleanAllClicked() {is=title K2Node_FunctionEntry}]}
 CleanCurSegmentsClicked {nodes [CleanCurSegmentsClicked() {is=title K2Node_FunctionEntry}]}
 Clear Current Segments {nodes [
   Clear Current Segments() {is=title K2Node_FunctionEntry}
   > Call Clean Cur Segments Clicked!(execute:Exec, self:BP STPS Travel Time Object Reference) -> (then:Exec) {is=title K2Node_CallDelegate}
  ]}
 Clear Other Splines {nodes [
   Clear Other Splines() {is=title K2Node_FunctionEntry}
   > Call Clean All Clicked!(execute:Exec, self:BP STPS Travel Time Object Reference) -> (then:Exec) {is=title K2Node_CallDelegate}
  ]}
 EventGraph {links [], nodes []}
]
variables [
 CleanAllClicked {default=()}
 CleanCurSegmentsClicked {default=()}
 Distance Units {type=E_TravelTime_DistUnits, default=cm}
 Follow Surface {type=E_TravelTime_ReprojectOptions, default=Off}
 Grid Snap Size {type=integer, default=16}
 Grid Snapping {type=boolean}
 Movement Presets {type=E_TravelTime_SpeedPresets, default=Run}
 Movement Speed {type=number, default=550}
 Movement Speed Multiplier {type=number, default=1}
 OpaqueLines {type=boolean}
 ReadoutOpenOnExit {type=boolean, default=true}
 SamplePointSize {type=number, default=75}
 Segment 1 Color {type=Color, default=<rgba 0.5216 0.5294 0.5333 1>}
 Segment 2 Color {type=Color, default=<rgba 0.2275 0.2275 0.2275 1>}
 Segment Line Thickness {type=number, default=1}
 ShortcutsOpenOnExit {type=boolean, default=true}
 Text Background Color {type=Color, default=<rgba 0 0 0 0.749>}
 Text Background Opacity {type=number, default=0.75}
 Text Size {type=number, default=1}
 TextColor {type=Color, default=<rgba 1 1 1 1>}
 Time Units {type=E_TravelTime_TimeUnits, default=sec}
 Use Straight Lines {type=boolean}
]
