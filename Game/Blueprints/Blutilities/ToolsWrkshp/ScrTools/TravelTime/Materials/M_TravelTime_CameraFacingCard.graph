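# Translucent unlit material for camera-facing text cards in the TravelTime Blutility tool, with depth testing disabled so the cards show through geometry.
# A shared billboard function aligns the mesh to the camera and receives a camera-distance scale with a per-axis multiplier.
# A sphere mask on the distance-field font darkens vertex color into an outline, then a Color parameter tints it.
# A second, UV-based camera-space billboard offset is also present.
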Material /Game/Blueprints/Blutilities/ToolsWrkshp/ScrTools/TravelTime/Materials/M_TravelTime_CameraFacingCard
graph {
 properties {
  bCastRayTracedShadows=false
  bDisableDepthTest=true
  blendMode=BLEND_Translucent
  pixelDepthOffsetMode=PDOM_Legacy
  shadingModel=MSM_Unlit
  usedShadingModels=Unlit
 }
 nodes [
  n1 = FontSampleParameter() {font=/Engine/EngineFonts/RobotoDistanceField, parameterName=Font}
  LinearInterpolate(Alpha=@n1.R) {constA=-90, constB=90}
  n2 = Divide(
   A=Length(None=WorldPosition() {worldPositionShaderOffset=WPT_CameraRelativeNoOffsets}.XYZ) {is=class MaterialExpressionLength}
   B=Divide(B=ScalarParameter() {defaultValue=1, parameterName=camdistscale}) {constA=2000}
  )
  MaterialFunctionCall!(
   PivotLocation
   camdistscale=Multiply(
    A=MaterialFunctionCall(X=@n2, Y=@n2, Z=@n2) {materialFunction=&MakeFloat3}.Result
    B=VectorParameter() {parameterName=scaleMult, defaultValue=<rgba 1 1 1 1>}.RGB
   )
   WS Normals
   WorldPosition=WorldPosition() {worldPositionShaderOffset=WPT_ExcludeAllShaderOffsets}.XYZ
   Custom Object Basis 1=Constant3Vector() {constant=<rgba 0 0 1 0>}
   Custom Object Basis 2=Constant3Vector() {constant=<rgba 1 0 0 0>}
   Custom Object Basis 3=Constant3Vector() {constant=<rgba 0 -1 0 0>}
  ) {materialFunction=/Game/Blueprints/Blutilities/ToolsWrkshp/ScrTools/Shared/SharedFunctions/AlignMeshToTheCamera_BillBoards}
  n3 = ScalarParameter() {defaultValue=0.442, parameterName=OutlineMask}
  n4 = ScalarParameter() {defaultValue=0.09, parameterName=OutlineThickness}
  n5 = ScalarParameter() {defaultValue=0.7, parameterName=OutlineHardness}
  Multiply(
   A=LinearInterpolate(A=VertexColor(), B=Constant3Vector(), Alpha=SphereMask(A=@n1.R, B=@n3, Radius=@n4, Hardness=@n5))
   B=VectorParameter() {parameterName=Color, defaultValue=<rgba 100 100 100 0>}.RGB
  )
  n6 = TextureCoordinate()
  n7 = ComponentMask(None=@n6) {g=true}
  n8 = Subtract(A=AppendVector(A=ComponentMask(None=@n6) {r=true}, B=OneMinus(None=@n7))) {constB=0.5}
  n9 = Max(A=SquareRoot(None=DotProduct(A=@n8, B=@n8))) {constB=0.0001}
  n10 = WorldPosition()
  n11 = ObjectPositionWS()
  n12 = AppendVector(A=Multiply(A=Divide(A=@n8, B=@n9), B=Distance(A=@n10.XYZ, B=@n11)), B=Constant())
  Subtract(A=Add(A=Transform(None=@n12) {transformSourceType=TRANSFORMSOURCE_Camera}, B=@n11), B=@n10.XYZ)
  VertexColor()
 ]
}
