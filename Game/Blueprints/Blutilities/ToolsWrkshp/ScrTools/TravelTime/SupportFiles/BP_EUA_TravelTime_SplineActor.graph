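# Editor utility spline actor supporting the TravelTime tool, holding data for measured lines and their text labels.
# Its variables store spline points, per-line and total lengths and run times, and text and background render components with colour and canvas layout settings.
# The event graph and construction script contain only empty entry nodes, so the actor has no logic of its own here.
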EditorUtilityBlueprint /Game/Blueprints/Blutilities/ToolsWrkshp/ScrTools/TravelTime/SupportFiles/BP_EUA_TravelTime_SplineActor
graphs [
 EventGraph {nodes [
   Event BeginPlay()
   Event ActorBeginOverlap()
   Event Tick!() -> (OutputDelegate, then, DeltaSeconds:float)
  ]}
 UserConstructionScript {nodes [Construction Script()]}
]
variables [
 AllBGRenderComponents {type="array<TextRenderComponent>", default []}
 AllTextRenderComponents {type="array<TextRenderComponent>", default []}
 CanvasRes {type=Vector2D, default=<1024 1024>}
 Centre X {type=boolean, default=true}
 Centre Y {type=boolean, default=true}
 CurTime {type=number}
 DynMatInstName {type=string, default=MeasureToolBillboardDynMat}
 LastWaitTime {type=number}
 LineNumber {type="array<integer>"}
 LinePoints {type="array<S_TravelTime_LinePoints>", default []}
 PlacedLocationBeforeOffset {type="array<Vector>"}
 Scale {type=Vector2D, default=<0.8 0.8>}
 Screen Position {type=Vector2D, default=<512 512>}
 SplinePointPositions {type="array<Vector>"}
 TextColour {type=Color, default=<rgba 0 0 0 0>}
 TextOutlineColour {type=Color, default=<rgba 0 0 0 0>}
 TextToRender {type="array<string>", default []}
 ThisLineLen {type="array<number>", default []}
 ThisLineRunTime {type="array<number>", default []}
 TotalLen {type=number}
 TotalRunTime {type=number}
]
