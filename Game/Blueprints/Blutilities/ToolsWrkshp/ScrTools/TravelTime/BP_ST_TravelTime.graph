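# Scriptable editor tool for drawing spline paths in a level and measuring their distance and travel time at a configurable movement speed.
# Clicks, hotkeys and per-point translation gizmos edit a spawned spline actor; editing a selected one rebuilds it from saved points.
# Spline, hidden reprojection spline and mesh components are added through SubobjectDataSubsystem.
# Watched, restorable property sets supply settings, while a header widget shows readouts, shortcuts and cleanup actions.
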EditorUtilityBlueprint /Game/Blueprints/Blutilities/ToolsWrkshp/ScrTools/TravelTime/BP_ST_TravelTime
graphs [
 AddActorSubobjects {nodes [
   AddActorSubobjects() {is=title K2Node_FunctionEntry}
   > n1 = Get Scene Root Handle!(execute, self) -> (then, Root Handle:Subobject Data Handle Structure) {is=title K2Node_CallFunction}
   > n2 = Sequence!(execute) -> (then_0, then_1, then_2:Exec, then_3:Exec, then_4:Exec)
   n3 = AddNewSubobject!(
    execute=@n2.then_1
    self:Subobject Data Subsystem Object Reference=Get SubobjectDataSubsystem()
    Params:"Add New Subobject Params Structure (by ref)"=Make Add New Subobject Params!(
     ParentHandle:Subobject Data Handle Structure=@n1
     NewClass:Object Class Reference=/Script/Engine.SplineComponent
     BlueprintContext:Blueprint Object Reference
     bSkipMarkBlueprintModified:bool=False
     bConformTransformToParent:bool=True
    ) -> (AddNewSubobjectParams:Add New Subobject Params Structure) {is=title K2Node_MakeStruct}
   ) -> (then, FailReason:Text, ReturnValue:Subobject Data Handle Structure)
   > n4 = Cast To SplineComponent!(
    execute
    Object=GetAssociatedObject!(Data:"Subobject Data Structure (by ref)"=GetData(DataHandle=@n3.ReturnValue)) -> (ReturnValue:Object Reference)
   ) -> (then, CastFailed, AsSpline Component:Spline Component Object Reference) {is=title K2Node_DynamicCast}
   > n5 = Set ReprojectSplineComponent(ReprojectSplineComponent:Spline Component Object Reference=@n4) -> (Output_Get:Spline Component Object Reference)
   > ClearSplinePoints!(execute, self:Spline Component Object Reference=@n5, bUpdateSpline:bool=true) -> (then)
   > SetVisibility!(execute, self:Scene Component Object Reference=@n5, bNewVisibility:bool=false, bPropagateToChildren:bool=true) -> (then)
   n6 = AddNewSubobject(
    execute=@n2.then_2
    self=Get SubobjectDataSubsystem()
    Params=Make Add New Subobject Params(ParentHandle=@n1, NewClass=/Script/GeometryFramework.DynamicMeshComponent, bSkipMarkBlueprintModified=False, bConformTransformToParent=True)
   )
   > n7 = Cast To DynamicMeshComponent!(execute, Object=GetAssociatedObject(Data=GetData(DataHandle=@n6.ReturnValue))) -> (then, CastFailed, AsDynamic Mesh Component:Dynamic Mesh Component Object Reference) {is=title K2Node_DynamicCast}
   > n8 = Set Dynamic Mesh Component(Dynamic Mesh Component:Dynamic Mesh Component Object Reference=@n7) -> (Output_Get:Dynamic Mesh Component Object Reference)
   > SetVisibility(self=@n8, bNewVisibility=true, bPropagateToChildren=false)
   n9 = AddNewSubobject(
    execute=@n2.then_3
    self=Get SubobjectDataSubsystem()
    Params=Make Add New Subobject Params(ParentHandle=@n1, NewClass=&StaticMeshComponent, bSkipMarkBlueprintModified=False, bConformTransformToParent=True)
   )
   > n10 = Cast To StaticMeshComponent(Object=GetAssociatedObject(Data=GetData(DataHandle=@n9.ReturnValue)))
   > n11 = Set Static Mesh Component(Static Mesh Component:Static Mesh Component Object Reference=@n10) -> (Output_Get:Static Mesh Component Object Reference)
   > SetVisibility(self=@n11, bNewVisibility=true, bPropagateToChildren=false)
   n12 = For Each Loop(
    Exec=@n2.then_4
    Array:Array of Object References=Make Array!(
     "[0]":Object Reference=Get SplineComponent() -> (SplineComponent:Spline Component Object Reference)
     "[1]":Object Reference=Get Dynamic Mesh Component() -> (Dynamic Mesh Component:Dynamic Mesh Component Object Reference)
     "[2]":Object Reference=Get Static Mesh Component() -> (Static Mesh Component:Static Mesh Component Object Reference)
     "[3]":Object Reference=Get ReprojectSplineComponent() -> (ReprojectSplineComponent:Spline Component Object Reference)
    ) -> (Array:Array of Object References)
   ) -> (Array Element:Object Reference)
   > SetComponentEditorOnly!(execute, self, Object:Object Reference=@n12) -> (then)
   n13 = AddNewSubobject(
    execute=@n2.then_0
    self=Get SubobjectDataSubsystem()
    Params=Make Add New Subobject Params(ParentHandle=@n1, NewClass=&SplineComponent, bSkipMarkBlueprintModified=False, bConformTransformToParent=True)
   )
   > n14 = Cast To SplineComponent(Object=GetAssociatedObject(Data=GetData(DataHandle=@n13.ReturnValue)))
   > n15 = Set SplineComponent(SplineComponent:Spline Component Object Reference=@n14) -> (Output_Get:Spline Component Object Reference)
   > ClearSplinePoints(self=@n15, bUpdateSpline=true)
  ]}
 AddPointGizmos {nodes [
   n1 = AddPointGizmos!() -> (then, Vizible:bool=true) {is=title K2Node_FunctionEntry}
   > Set LocViz(LocViz:bool=@n1) -> (Output_Get:bool)
   > n2 = For Each Loop(Array:Array of Strings=Get AllGizmoNames() -> (AllGizmoNames:Array of Strings)) -> (Array Element:String)
   Clear(execute=@n2.Completed, TargetArray=Get AllGizmoNames())
   n3 = Get SplineComponent()
   > n4 = For Loop(LastIndex=int - int(A=GetNumberOfSplinePoints!(self:Spline Component Object Reference=@n3) -> (ReturnValue:int=0)))
   n5 = Get SplineComponent()
   n6 = Append(A=Get GizmoBaseIdent() -> (GizmoBaseIdent:String), B="To String (Integer)"(InInt=@n4))
   n7 = Get Settings() -> (Settings:BP STPS Travel Time Object Reference)
   > CreateTRSGizmo!(
    execute
    self:Scriptable Interactive Tool Object Reference
    Identifier:String=@n6
    InitialTransform:Transform="To Transform (Vector)"!(InLocation:Vector=GetLocationAtSplinePoint!(self:Spline Component Object Reference=@n5, PointIndex:int=@n4, CoordinateSpace:ESplineCoordinateSpace Enum=World) -> (ReturnValue:Vector)) -> (ReturnValue:Transform)
    GizmoOptions:Scriptable Tool Gizmo Options Structure=Make Scriptable Tool Gizmo Options!(
     GizmoMode:EScriptableToolGizmoMode Enum=TranslationOnly
     TranslationParts:Bitmask Integer=255
     RotationParts:Bitmask Integer=241
     ScaleParts:Bitmask Integer=1
     CoordSystem:EScriptableToolGizmoCoordinateSystem Enum=World
     bSnapTranslation:bool=Get Grid Snapping!(self:BP STPS Travel Time Object Reference=@n7) -> (Grid Snapping:bool)
     bSnapRotation:bool=false
     bRepositionable:bool=False
     bAllowNegativeScaling:bool=false
    ) -> (ScriptableToolGizmoOptions:Scriptable Tool Gizmo Options Structure) {is=title K2Node_MakeStruct}
   ) -> (Success:Exec, Failure:Exec)
   > Add Unique(TargetArray=Get AllGizmoNames(), NewItem=@n6)
   > SetGizmoVisible!(execute, self:Scriptable Interactive Tool Object Reference, Identifier:String=@n6, bVisible:bool=Get LocViz() -> (LocViz:bool)) -> (then)
   DestroyTRSGizmo!(execute=@n2.LoopBody, self:Scriptable Interactive Tool Object Reference, Identifier:String=@n2) -> (Success:Exec, Failure:Exec)
  ]}
 AddSplineActor {nodes [
   AddSplineActor() {is=title K2Node_FunctionEntry}
   n1 = int + int(
    A=Length(TargetArray:Array of BP EUA Travel Time Spline Actor Object References=Get All Measure Tool Spline Actors EUA() -> (All Measure Tool Spline Actors EUA:Array of BP EUA Travel Time Spline Actor Object References))
    B=Get CurNumericSuffix() -> (CurNumericSuffix:int=0)
   )
   > Set SplineActorLabel(SplineActorLabel:String=Append!(A=Get SplineActorLabelRoot() -> (SplineActorLabelRoot:String), B=" ", C:String="To String (Integer)"(InInt=@n1)) -> (ReturnValue)) -> (Output_Get:String)
   > n2 = SpawnActorFromClass!(
    execute
    self:Editor Actor Subsystem Object Reference=Get EditorActorSubsystem()
    ActorClass:Actor Class Reference=/Game/Blueprints/Blutilities/ToolsWrkshp/ScrTools/TravelTime/SupportFiles/BP_EUA_TravelTime_SplineActor.BP_EUA_TravelTime_SplineActor_C
    Location:Vector="0, 0, 0"
    Rotation:Rotator
    bTransient:bool=Get Transient() -> (Transient:bool)
   ) -> (then, ReturnValue:BP EUA Travel Time Spline Actor Object Reference)
   > n3 = Set SplineActorEUA(SplineActorEUA:BP EUA Travel Time Spline Actor Object Reference=@n2) -> (Output_Get:BP EUA Travel Time Spline Actor Object Reference)
   > SetEditorProperty(Object=@n3, PropertyName=bIsEditorOnlyActor, PropertyValue=Get EditorOnly() -> (EditorOnly:bool))
   > SetActorLabel(self=@n3, NewActorLabel=Get SplineActorLabel() -> (SplineActorLabel:String), bMarkDirty=false)
   > n4 = Sequence!(execute) -> (then_0, then_1, then_2)
   > AddActorSubobjects() {is=title K2Node_CallFunction}
   Return Node!(
    execute=@n4.then_2
    SplineActor:BP EUA Travel Time Spline Actor Object Reference=Get SplineActorEUA() -> (SplineActorEUA:BP EUA Travel Time Spline Actor Object Reference)
   )
   Add Unique(
    execute=@n4.then_1
    TargetArray:Array of BP EUA Travel Time Spline Actor Object References=Get All Measure Tool Spline Actors EUA()
    NewItem:"BP EUA Travel Time Spline Actor Object Reference (by ref)"=Get SplineActorEUA()
   )
  ]}
 CalcLineCentre {nodes [
   n1 = CalcLineCentre!() -> (then, Point0:Vector=~, Point1:Vector=~) {is=title K2Node_FunctionEntry}
   n2 = Make Literal Float(Value=0.5)
   n3 = BreakVector(InVec=@n1.Point1)
   n4 = float * float(A=@n2, B=@n3.Z) {is=title K2Node_PromotableOperator}
   n5 = BreakVector(InVec=@n1)
   n6 = float - float(A=Make Literal Float(Value=1), B=@n2)
   n7 = float * float(A=@n6, B=@n5.Z)
   n8 = float * float(A=@n6, B=@n5)
   n9 = float * float(A=@n2, B=@n3)
   n10 = float * float(A=@n6, B=@n5.Y)
   n11 = float * float(A=@n2, B=@n3.Y)
   > Return Node!(execute, CentrePoint:Vector=MakeVector(X=float + float(A=@n8, B=@n9), Y=float + float(A=@n10, B=@n11), Z=float + float(A=@n7, B=@n4)))
  ]}
 Click Capture Check {nodes [
   n1 = Click Capture Check!() -> (
    then
    InputDeviceState_bShiftKeyDown:bool=~
    InputDeviceState_bAltKeyDown:bool=~
    InputDeviceState_bCtrlKeyDown:bool=~
    InputDeviceState_bCmdKeyDown:bool=~
    InputDeviceState_Keyboard_ActiveKey_Button:Key Structure
    InputDeviceState_Keyboard_ActiveKey_bPressed:bool=~
    InputDeviceState_Keyboard_ActiveKey_bDown:bool=~
    InputDeviceState_Keyboard_ActiveKey_bReleased:bool=~
    InputDeviceState_Keyboard_ActiveKey_bDoubleClicked:bool=~
    InputDeviceState_Mouse:Mouse Input Device State Structure
   ) {is=title K2Node_FunctionEntry}
   n2 = Get isEditHotkeyDown() -> (isEditHotkeyDown:bool)
   > Return Node!(execute, ReturnValue:bool=AND Boolean(A=NOT Boolean(A=@n1.InputDeviceState_bAltKeyDown), B=NOT Boolean(A=@n2)))
  ]}
 ConvertSplineToPolyline {nodes [
   ConvertSplineToPolyline!() -> (then, TargetArray:Array of Vectors) {is=title K2Node_FunctionEntry}
   n1 = Get ReprojectSplineComponent()
   > Branch(Condition="integer > integer"(A=GetNumberOfSplinePoints(self=@n1) -> (ReturnValue=0), B=1))
   n2 = GetSplineLength!(self:Spline Component Object Reference=@n1) -> (ReturnValue:float)
   n3 = Get PolyLineSpacing() -> (PolyLineSpacing:double)
   > n4 = ConvertSplineToPolyPath!(
    execute
    Spline:Spline Component Object Reference=@n1
    SamplingOptions_NumSamples:int=Truncate(A=float / float(A=@n2, B=@n3))
    SamplingOptions_ErrorTolerance:float=1
    SamplingOptions_SampleSpacing:EGeometryScriptSampleSpacing Enum=UniformTime
    SamplingOptions_CoordinateSpace:ESplineCoordinateSpace Enum=World
    SamplingOptions_RangeMethod:EGeometryScriptEvaluateSplineRange Enum=FullSpline
    SamplingOptions_RangeStart:float=0
    SamplingOptions_RangeEnd:float=1
   ) -> (then, PolyPath:Poly Path Structure)
   > n5 = ConvertPolyPathToArray!(execute, PolyPath:Poly Path Structure=@n4) -> (then, VertexArray:Array of Vectors)
   > n6 = For Each Loop(Array:Array of Vectors=@n5) -> (Array Element:Vector)
   > Add(TargetArray:Array of Vectors=Get TargetArray() -> (TargetArray:Array of Vectors), NewItem:"Vector (by ref)"=@n6) {is=title K2Node_CallArrayFunction}
   Get SplineComponent()
  ]}
 DeleteHotkeyCaptureCheck {nodes [
   DeleteHotkeyCaptureCheck!() -> (then, InputDeviceState:"Input Device State Structure (by ref)") {is=title K2Node_FunctionEntry}
   > Return Node(ReturnValue=true)
  ]}
 DestroyPointGizmos {nodes [
   DestroyPointGizmos() {is=title K2Node_FunctionEntry}
   > n1 = For Each Loop(Array:Array of Strings=Get AllGizmoNames()) -> (Array Element:String)
   Clear(execute=@n1.Completed, TargetArray=Get AllGizmoNames())
   DestroyTRSGizmo(execute=@n1.LoopBody, Identifier=@n1)
  ]}
 EventGraph {nodes [
   n1 = Get Clear Warning After Time() -> (Clear Warning After Time:double)
   n2 = Get Last Clear Time() -> (Last Clear Time:double)
   n3 = Get CurTime()
   n4 = TimedClearUserWarning!(execute, self, ClearAfterTime:double=@n1, CurTime:double=@n3, LastClearTime:double=@n2) -> (then, NewLastClearTime:double)
   n5 = TimedClearUserWarning!(
    execute=@n13.then_2
    ClearAfterTime=@n1
    CurTime=@n3
    LastClearTime=@n2
    Scriptable Tool Obj Ref:Scriptable Interactive Tool Object Reference=Self-Reference()
   ) -> (then, NewLastClearTime)
   Set Last Clear Time(execute=[@n4.then, @n5.then], Last Clear Time:double=@n5) -> (Output_Get:double)
   Event OnScriptSetup()
   > n6 = Sequence!(execute) -> (then_0, then_1, then_2, then_3:Exec, then_4:Exec, then_5:Exec, then_6:Exec, then_7:Exec, then_8:Exec)
   > K2Node_Composite_0 = InitCommonVars!(execute) -> (then:Exec) {is=title K2Node_Composite}
   K2Node_Composite_4 = Add Property Sets!(execute=@n6.then_2) -> (then:Exec) {is=title K2Node_Composite}
   n7 = ShouldDoEdit!(execute=@n6.then_3, self) -> (then, ShouldDoEdit:bool, Selected Spline Actor:BP EUA Travel Time Spline Actor Object Reference)
   > Branch(Condition=@n7)
   > K2Node_Composite_1 = Set up Editing Existing!(execute, InputPin:BP EUA Travel Time Spline Actor Object Reference=@n7.Selected Spline Actor) -> (Completed:Exec) {is=title K2Node_Composite}
   K2Node_Composite_3 = Add Behaviours!(InputPin:Exec=@n6.then_5) -> (then:Exec) {is=title K2Node_Composite}
   Bind Event to Clean All Clicked(execute=@n6.then_6, self:BP STPS Travel Time Object Reference=Get Settings(), Delegate=Create Event()) {is=title K2Node_AddDelegate}
   > Bind Event to Clean Cur Segments Clicked(self:BP STPS Travel Time Object Reference=Get Settings(), Delegate=Create Event()) {is=title K2Node_AddDelegate}
   K2Node_Composite_6 = Watch Properties!(execute=@n6.then_7) -> (then_4:Exec) {is=title K2Node_Composite}
   Branch(execute=@n6.then_8, Condition=Get Is Game View On Tool Start() -> (Is Game View On Tool Start:bool))
   > Set HaveAlreadyWarnedGameView(HaveAlreadyWarnedGameView:bool=false) -> (Output_Get:bool)
   > GetSetGameView!(
    execute
    self:BP Travel Time Editor Only Utils Object Reference=Get EditorOnlyTools() -> (EditorOnlyTools:BP Travel Time Editor Only Utils Object Reference)
    SetGameView:bool=true
    GameViewVal:bool=false
   ) -> (then, ViewportKey:Name=None, IsGameView:bool)
   n8 = Construct Object from Class(
    execute=@n6.then_1
    Class=&BP_UW_SharedHeaderWidget_C
   )
   > n9 = Set Header Widget(Header Widget:BP UW Shared Header Widget Object Reference=@n8) -> (Output_Get:BP UW Shared Header Widget Object Reference)
   > SetToolPanelToolHeaderWidget(Widget=@n9)
   n10 = Branch(execute=@n6.then_4, Condition=Get DoEdit() -> (DoEdit:bool))
   K2Node_Composite_2 = Set up Create New!(execute=@n10.else) -> (then:Exec) {is=title K2Node_Composite}
   Event OnScriptShutdown()
   > OnScriptShudown()
   n11 = Event OnGizmoTransformChanged!() -> (OutputDelegate, then, GizmoIdentifier:String, NewTransform:Transform) {is=title K2Node_Event}
   > GizmoTransformChanged!(execute, self, Gizmo Ident:String=@n11.GizmoIdentifier, New Transform:Transform=@n11.NewTransform) -> (then)
   n12 = Event OnScriptTick()
   > Set CurTime(CurTime=float + float(A=@n12.DeltaTime, B=Get CurTime()))
   > n13 = Sequence!(execute) -> (then_0, then_1, then_2)
   n14 = Get HaveAlreadyWarnedGameView() -> (HaveAlreadyWarnedGameView:bool)
   > Branch(Condition=AND Boolean(A=Get Is Game View On Tool Start(), B=NOT Boolean(A=@n14)))
   > WarnGameViewOn()
   n15 = Branch(execute=@n13.then_1, Condition=Get HaveAlreadyInitHeader() -> (HaveAlreadyInitHeader:bool))
   Set HaveAlreadyInitHeader(execute=@n15.else, HaveAlreadyInitHeader:bool=true) -> (Output_Get:bool)
   > n16 = Sequence!(execute) -> (then_0, then_1, then_2, then_3:Exec)
   n17 = Get Header Widget()
   > n18 = AddExpandableBox(self=@n17)
   > n19 = Set Readout Box(Readout Box:BP UW Expandable Box Object Reference=@n18) -> (Output_Get:BP UW Expandable Box Object Reference)
   > SetUpBox(
    self=@n19
    Name=Get Readout Box Label!(self:BP UW Shared Header Widget Object Reference=@n17) -> (Readout Box Label:Text)
    OpenByDefault=Get ReadoutOpenOnExit!(self:BP STPS Shared Widget State Object Reference=Get Shared State Settings() -> (Shared State Settings:BP STPS Shared Widget State Object Reference)) -> (ReadoutOpenOnExit:bool)
   )
   > n20 = Sequence!(execute) -> (then_0, then_1)
   n21 = Set Distance Line(execute=@n22.then, Distance Line:BP UW Readout Line Object Reference=@n22) -> (Output_Get:BP UW Readout Line Object Reference)
   n22 = SetupLine!(
    execute=@n20.then_0
    self:BP UW Expandable Box Object Reference=Get Readout Box() -> (Readout Box:BP UW Expandable Box Object Reference)
    CurrentLine:BP UW Readout Line Object Reference=@n21
    DynamicLine:Vertical Box Object Reference
    Label:Text=Get Widget Distance Line Label() -> (Widget Distance Line Label:Text)
    Value:Text=Get Widget Distance Line Value() -> (Widget Distance Line Value:Text)
   ) -> (then, Readout Line:BP UW Readout Line Object Reference)
   n23 = Set Time Line(execute=@n24.then, Time Line:BP UW Readout Line Object Reference=@n24) -> (Output_Get:BP UW Readout Line Object Reference)
   n24 = SetupLine(
    execute=@n20.then_1
    self=Get Readout Box()
    CurrentLine=@n23
    Label=Get Widget Time Line Label() -> (Widget Time Line Label:Text)
    Value=Get Widget Time Line Value() -> (Widget Time Line Value:Text)
   )
   n25 = Get Header Widget()
   n26 = AddExpandableBox(execute=@n16.then_1, self=@n25)
   > n27 = Set Shortcut Box(Shortcut Box:BP UW Expandable Box Object Reference=@n26) -> (Output_Get:BP UW Expandable Box Object Reference)
   > n28 = SetUpBox(
    self=@n27
    Name=Get Shortcut Box Label!(self:BP UW Shared Header Widget Object Reference=@n25) -> (Shortcut Box Label:Text)
    OpenByDefault=Get ShortcutsOpenOnExit!(self:BP STPS Shared Widget State Object Reference=Get Shared State Settings()) -> (ShortcutsOpenOnExit:bool)
   )
   n29 = Set ShortcutLines(execute=@n30.then, ShortcutLines:Array of BP UW Readout Line Object References=@n30) -> (Output_Get:Array of BP UW Readout Line Object References)
   n30 = AddShortcuts!(
    execute=@n28.then
    self:BP UW Expandable Box Object Reference=Get Shortcut Box() -> (Shortcut Box:BP UW Expandable Box Object Reference)
    ShortcutMap:Map of Strings to Strings=Get Shortcut Map() -> (Shortcut Map:Map of Strings to Strings)
    ExistingShortcutLines:Array of BP UW Readout Line Object References=@n29
    ShortcutMapKeys:Array of Strings=Get Shortcut Map Keys() -> (Shortcut Map Keys:Array of Strings)
   ) -> (then, AllShortcutLines:Array of BP UW Readout Line Object References)
   SetOverlayInfo!(
    execute=@n16.then_2
    self
    TextToOverlay:Text
    IsWarning:bool=false
    UseMessageLine:bool=false
    Total Dist:double=Get TotalDist() -> (TotalDist:double)
    Total Time:double=Get TotalTime() -> (TotalTime:double)
   ) -> (then)
   > SetOverlayInfo(IsWarning=true, UseMessageLine=false, Total Dist=0, Total Time=0)
   > Set Last Clear Time(Last Clear Time=Get CurTime())
   n31 = AddExpandableBox(execute=@n16.then_3, self=Get Header Widget())
   > n32 = Set Actions Box(Actions Box=@n31)
   > SetUpBox(self=@n32, Name=Get Actions Box Label(self=Get Header Widget()))
   > n33 = Construct Object from Class(Class=/Game/Blueprints/Blutilities/ToolsWrkshp/ScrTools/TravelTime/SupportFiles/BP_UW_ActionsButton_TravelTime.BP_UW_ActionsButton_TravelTime_C) -> (ReturnValue:BP UW Actions Button Travel Time Object Reference)
   > n34 = Set ActionButtons(ActionButtons:BP UW Actions Button Travel Time Object Reference=@n33) -> (Output_Get:BP UW Actions Button Travel Time Object Reference)
   > AddChildToVerticalBox(self=Get DynamicEntries(self=Get Actions Box()), Content=@n34)
   > Bind Event to Event Clear Cur Segs Clicked(
    self:BP UW Actions Button Travel Time Object Reference=Get ActionButtons() -> (ActionButtons:BP UW Actions Button Travel Time Object Reference)
    Delegate=Create Event()
   ) {is=title K2Node_AddDelegate}
   > Bind Event to Event Clear Other Splines Clicked(self:BP UW Actions Button Travel Time Object Reference=Get ActionButtons(), Delegate=Create Event()) {is=title K2Node_AddDelegate}
   n35 = Event OnGizmoTransformStateChange!() -> (OutputDelegate, then, GizmoIdentifier:String, CurrentTransform:Transform, ChangeType:EScriptableToolGizmoStateChangeType Enum=BeginTransform) {is=title K2Node_Event}
   > GizmoTransformStateChange!(
    execute
    self
    GizIdentifier:String=@n35.GizmoIdentifier
    NewTransform:Transform=@n35.CurrentTransform
    ChangeType:EScriptableToolGizmoStateChangeType Enum=@n35.ChangeType
   ) -> (then)
  ]}
 InitCommonVars {
  path=EventGraph.K2Node_Composite_0.InitCommonVars
  nodes [
   Inputs()
   > n1 = Sequence!(execute) -> (then_0, then_1, then_2:Exec, then_3:Exec, then_4:Exec, then_5:Exec)
   n2 = Self-Reference()
   Set TransactionContext(execute=@n1.then_1, TransactionContext:String="To String (Text)"(InText=Get ToolLongName(None=@n2) -> (ToolLongName:Text))) -> (Output_Get:String)
   n3 = Now!() -> (
    ReturnValue_Year:int=0
    ReturnValue_Month:int=0
    ReturnValue_Day:int=0
    ReturnValue_Hour:int=0
    ReturnValue_Minute:int=0
    ReturnValue_Second:int=0
    ReturnValue_Millisecond:int=0
   )
   > Set ToolStartTime(ToolStartTime:String="To String (Text)"(InText=Format Text!(
      Format="LOCTEXT(\"{y}_{m}_{d}_{h}_{min}_{s}\")"
      y:int=@n3
      m:int=@n3.ReturnValue_Month
      d:int=@n3.ReturnValue_Day
      h:int=@n3.ReturnValue_Hour
      min:int=@n3.ReturnValue_Minute
      s:int=@n3.ReturnValue_Second
     ) -> (Result))) -> (Output_Get:String)
   n4 = Construct Object from Class(
    execute=@n1.then_2
    Class=/Game/Blueprints/Blutilities/ToolsWrkshp/ScrTools/TravelTime/SupportFiles/BP_TravelTime_EditorOnlyUtils.BP_TravelTime_EditorOnlyUtils_C
   ) -> (ReturnValue:BP Travel Time Editor Only Utils Object Reference)
   > n5 = Set EditorOnlyTools(EditorOnlyTools:BP Travel Time Editor Only Utils Object Reference=@n4) -> (Output_Get:BP Travel Time Editor Only Utils Object Reference)
   > n6 = GetSetGameView(self=@n5, SetGameView=false, GameViewVal=false) -> (ViewportKey=None)
   > Set Is Game View On Tool Start(Is Game View On Tool Start:bool=@n6.IsGameView) -> (Output_Get:bool)
   Clear(
    execute=@n1.then_3
    TargetArray:Array of Actor Object References=Get AllNonLandscapeActors() -> (AllNonLandscapeActors:Array of Actor Object References)
   )
   > n7 = GetAllLevelActors(self=Get EditorActorSubsystem())
   > n8 = For Each Loop(Array:Array of Actor Object References=@n7) -> (Array Element:Actor Object Reference)
   n9 = Get Class(Object=@n8)
   > n10 = Branch(Condition=Contains(SearchIn=Get Class Display Name(Class=@n9), Substring=landscape))
   Add(execute=@n10.else, TargetArray:Array of Actor Object References=Get AllNonLandscapeActors(), NewItem:"Actor Object Reference (by ref)"=@n8)
   n11 = GetAllActorsOfClass!(
    execute=@n1.then_4
    ActorClass:Actor Class Reference=&BP_EUA_TravelTime_SplineActor_C
   ) -> (then, OutActors:Array of BP EUA Travel Time Spline Actor Object References)
   > Set All Measure Tool Spline Actors EUA(All Measure Tool Spline Actors EUA:Array of BP EUA Travel Time Spline Actor Object References=@n11) -> (Output_Get:Array of BP EUA Travel Time Spline Actor Object References)
   Outputs!(then:Exec=@n1.then_5)
  ]
 }
 Set up Editing Existing {
  path=EventGraph.K2Node_Composite_1.Set up Editing Existing
  nodes [
   n1 = Inputs!() -> (execute, InputPin:BP EUA Travel Time Spline Actor Object Reference)
   > Set SplineActorEUA(SplineActorEUA=@n1)
   > n2 = Sequence!(execute) -> (then_0, then_1, then_2, then_3, then_4, then_5, then_6:Exec, then_7:Exec)
   > Clear(TargetArray=Get ActorsSelectedAtStart() -> (ActorsSelectedAtStart:Array of Actor Object References))
   > n3 = GetSelectedLevelActors(self=Get EditorActorSubsystem())
   > n4 = For Each Loop(Array=@n3)
   SetSelectedLevelActors(execute=@n4.Completed, self=Get EditorActorSubsystem(), ActorsToSelect=Get ActorsSelectedAtStart())
   > Set IsEUAselected(IsEUAselected:bool=false) -> (Output_Get:bool)
   n5 = Branch(execute=@n4.LoopBody, Condition="Equal (Object)"(A=@n4, B=Get SplineActorEUA()))
   Add(execute=@n5.else, TargetArray=Get ActorsSelectedAtStart(), NewItem=@n4)
   n6 = Get SplineActorEUA()
   Set TotalTime(execute=@n2.then_2, TotalTime:double=Get TotalRunTime!(self:BP EUA Travel Time Spline Actor Object Reference=@n6) -> (TotalRunTime:double)) -> (Output_Get:double)
   > Set TotalDist(TotalDist:double=Get TotalLen!(self:BP EUA Travel Time Spline Actor Object Reference=@n6) -> (TotalLen:double)) -> (Output_Get:double)
   Destroy Actor(execute=@n2.then_3, self=Get SplineActorEUA())
   AddSplineActor!(execute=@n2.then_4, self) -> (then, SplineActor:BP EUA Travel Time Spline Actor Object Reference) {is=title K2Node_CallFunction}
   RefreshAll!(execute=@n2.then_5, self, ImmediateUpdate:bool=true) -> (then)
   Outputs!(Completed:Exec=@n2.then_6)
   Set AllSplinePoints(
    execute=@n2.then_1
    AllSplinePoints:Array of Vectors=Get SplinePointPositions!(self:BP EUA Travel Time Spline Actor Object Reference=Get SplineActorEUA()) -> (SplinePointPositions:Array of Vectors)
   ) -> (Output_Get:Array of Vectors)
   Reroute Node(InputPin:Wildcard) -> (OutputPin:Wildcard)
  ]
 }
 Set up Create New {
  path=EventGraph.K2Node_Composite_2.Set up Create New
  nodes [
   Inputs!() -> (execute)
   > n1 = Sequence!(execute) -> (then_0, then_1, then_2, then_3)
   > AddSplineActor()
   SetDynMeshMaterial(execute=@n1.then_2)
   Outputs!(then:Exec=@n1.then_3)
   Set TotalTime(execute=@n1.then_1, TotalTime=0)
   > Set TotalDist(TotalDist=0)
  ]
 }
 Add Behaviours {
  path=EventGraph.K2Node_Composite_3.Add Behaviours
  nodes [
   Inputs!() -> (InputPin:Exec)
   > AddMouseHoverBehavior!(
    execute
    self:Scriptable Modular Behavior Tool Object Reference
    BeginHoverSequenceHitTest:Delegate=Create Event()
    OnBeginHover:Delegate
    OnUpdateHover:Delegate
    OnEndHover:Delegate
    HoverCaptureCheck:Delegate
    CapturePriority:int=0
   ) -> (then)
   n1 = Create Event()
   n2 = Create Event()
   n3 = Create Event()
   > AddSingleClickBehavior!(
    execute
    self:Scriptable Modular Behavior Tool Object Reference
    IfHitByClick:Delegate=@n1
    OnHitByClick:Delegate=@n2
    CaptureCheck:Delegate=@n3
    CapturePriority:int=0
    MouseButton:EScriptableToolMouseButton Enum=MiddleButton
    bHitTestOnRelease:bool=true
   ) -> (then)
   > AddSingleClickBehavior(IfHitByClick=@n1, OnHitByClick=@n2, CaptureCheck=@n3, CapturePriority=0, MouseButton=LeftButton, bHitTestOnRelease=true)
   n4 = Create Event()
   n5 = Create Event()
   > AddSingleKeyInputBehavior!(
    execute
    self:Scriptable Modular Behavior Tool Object Reference
    OnKeyPressed:Delegate=Create Event()
    OnKeyReleased:Delegate=Create Event()
    OnForceEndCaptureFuncIn:Delegate=@n4
    Key:Key Structure=Get EditModeHotkey() -> (EditModeHotkey:Key Structure)
    CaptureCheck:Delegate=@n5
    CapturePriority:int=0
   ) -> (then)
   > AddSingleKeyInputBehavior(
    OnKeyPressed=Create Event()
    OnKeyReleased=Create Event()
    OnForceEndCaptureFuncIn=@n4
    Key=Get DeleteHotkey() -> (DeleteHotkey:Key Structure)
    CaptureCheck=@n5
    CapturePriority=0
   )
   > AddSingleKeyInputBehavior(
    OnKeyPressed=Create Event()
    OnKeyReleased=Create Event()
    OnForceEndCaptureFuncIn=@n4
    Key=Get SplitHotkey() -> (SplitHotkey:Key Structure)
    CaptureCheck=@n5
    CapturePriority=0
   )
   > AddSingleKeyInputBehavior(
    OnKeyPressed=Create Event()
    OnKeyReleased=Create Event()
    OnForceEndCaptureFuncIn=@n4
    Key=Get PrependHotkey() -> (PrependHotkey:Key Structure)
    CaptureCheck=@n5
    CapturePriority=0
   )
   > Outputs()
  ]
 }
 Add Property Sets {
  path=EventGraph.K2Node_Composite_4.Add Property Sets
  nodes [
   Inputs!() -> (execute:Exec)
   > n1 = Sequence!(execute) -> (then_0, then_1, then_2, then_3, then_4:Exec, then_5:Exec)
   > n2 = AddPropertySetOfType(
    PropertySetType=&/BP_STPS_TravelTime.BP_STPS_TravelTime_C
    Identifier=Get SettingsIdentifier() -> (SettingsIdentifier:String)
   )
   > n3 = Cast To BP_STPS_TravelTime!(execute, Object=@n2) -> (then, CastFailed, AsBP STPS Travel Time:BP STPS Travel Time Object Reference) {is=title K2Node_DynamicCast}
   > n4 = Set Settings(Settings:BP STPS Travel Time Object Reference=@n3) -> (Output_Get:BP STPS Travel Time Object Reference)
   > RestorePropertySetSettings(PropertySet=@n4, SaveKey=Get PropSetSaveRestoreKey() -> (PropSetSaveRestoreKey:String))
   Set PreviousMovementPreset(
    execute=@n1.then_1
    PreviousMovementPreset:E_TravelTime_SpeedPresets Enum=Get Movement Presets!(self:BP STPS Travel Time Object Reference=Get Settings()) -> (Movement Presets:E_TravelTime_SpeedPresets Enum=NewEnumerator0)
   ) -> (Output_Get:E_TravelTime_SpeedPresets Enum=NewEnumerator0)
   > Set Previous Movement Speed(Previous Movement Speed:double=Get Movement Speed!(self:BP STPS Travel Time Object Reference=Get Settings()) -> (Movement Speed:double)) -> (Output_Get:double)
   n5 = AddPropertySetOfType(
    execute=@n1.then_2
    PropertySetType=/Game/Blueprints/Blutilities/ToolsWrkshp/ScrTools/Shared/SharedPropertySets/BP_STPS_SharedWidgetState.BP_STPS_SharedWidgetState_C
    Identifier=Get SharedPropSetWidgetStateIdent() -> (SharedPropSetWidgetStateIdent:String)
   )
   > n6 = Cast To BP_STPS_SharedWidgetState!(execute, Object=@n5) -> (then, CastFailed, AsBP STPS Shared Widget State:BP STPS Shared Widget State Object Reference) {is=title K2Node_DynamicCast}
   > Set Shared State Settings(Shared State Settings:BP STPS Shared Widget State Object Reference=@n6) -> (Output_Get:BP STPS Shared Widget State Object Reference)
   > SetPropertySetVisibleByName!(execute, self:Scriptable Interactive Tool Object Reference, Identifier:String=Get SharedPropSetWidgetStateIdent(), bVisible:bool=false) -> (then)
   > RestorePropertySetSettings(PropertySet=Get Shared State Settings(), SaveKey=Get PropSetSaveRestoreKey())
   Outputs(then=@n1.then_5)
  ]
 }
 Watch Properties {
  path=EventGraph.K2Node_Composite_6.Watch Properties
  nodes [
   Inputs()
   > n1 = Sequence!(execute) -> (then_0, then_1, then_2, then_3, then_4, then_5, then_6:Exec, then_7:Exec, then_8:Exec)
   n2 = For Each Loop(
    Exec=@n1.then_7
    Array:Array of Strings=Make Array!("[0]":String=Distance Units, "[1]":String=Time Units, "[2]":String=Grid Snap Size) -> (Array:Array of Strings)
   ) -> (Array Element:String)
   > WatchProperty(PropertySet=Get Settings(), PropertyName=@n2, OnModified=Create Event())
   Outputs!(then_4:Exec=@n1.then_8)
   WatchProperty(execute=@n1.then_6, PropertySet=Get Settings(), PropertyName=Grid Snapping, OnModified=Create Event())
   n3 = WatchFloatProperty!(
    execute=@n1.then_2
    self:Scriptable Interactive Tool Object Reference
    PropertySet:Scriptable Interactive Tool Property Set Object Reference=Get Settings()
    PropertyName:String=Text Size
    OnModified:"Delegate (by ref)"=Create Event()
   ) -> (then, ReturnValue:Scriptable Interactive Tool Property Set Object Reference)
   > n4 = WatchFloatProperty(PropertySet=@n3, PropertyName=Movement Speed, OnModified=Create Event())
   > WatchFloatProperty(PropertySet=@n4, PropertyName=Movement Speed Multiplier, OnModified=Create Event())
   n5 = For Each Loop(Exec=@n1.then_1, Array=Make Array!("[0]"=Use Straight Lines) -> (Array))
   > WatchBoolProperty!(
    execute
    self:Scriptable Interactive Tool Object Reference
    PropertySet:Scriptable Interactive Tool Property Set Object Reference=Get Settings()
    PropertyName:String=@n5
    OnModified:"Delegate (by ref)"=Create Event()
   ) -> (then, ReturnValue:Scriptable Interactive Tool Property Set Object Reference)
   n6 = For Each Loop(Exec=@n1.then_3, Array=Make Array("[0]"=Segment Line Thickness))
   > WatchFloatProperty(PropertySet=Get Settings(), PropertyName=@n6, OnModified=Create Event())
   WatchProperty(execute=@n1.then_4, PropertySet=Get Settings(), PropertyName=Movement Presets, OnModified=Create Event())
   n7 = For Each Loop(
    Exec=@n1.then_5
    Array=Make Array!(
     "[0]"=TextColor
     "[1]":String=Text Background Color
     "[2]":String=Segment 1 Color
     "[3]":String=Segment 2 Color
     "[4]":String=Follow Surface
     "[5]":String=Text Background Opacity
    ) -> (Array)
   )
   > WatchProperty(PropertySet=Get Settings(), PropertyName=@n7, OnModified=Create Event())
  ]
 }
 Get Scene Root Handle {nodes [
   Get Scene Root Handle() {is=title K2Node_FunctionEntry}
   > n1 = Gather Subobject Data For Instance(self=Get SubobjectDataSubsystem(), Context=Get SplineActorEUA())
   > Set Parent Handle(Parent Handle:Subobject Data Handle Structure="Get (a copy)"(Array:Array of Subobject Data Handle Structures=@n1) -> (Output:Subobject Data Handle Structure)) -> (Output_Get:Subobject Data Handle Structure)
   > n2 = Set DefSceneRootHandle(DefSceneRootHandle:Subobject Data Handle Structure="Get (a copy)"(Array=@n1, Dimension 1=1)) -> (Output_Get:Subobject Data Handle Structure)
   > Return Node!(execute, Root Handle:Subobject Data Handle Structure=@n2)
  ]}
 GetGizmoDigit {nodes [
   n1 = GetGizmoDigit!() -> (then, GizID:String) {is=title K2Node_FunctionEntry}
   > Return Node!(execute, Digit:int=String To Integer(InString=Split(SourceString=@n1, InStr=_, SearchDir=FromEnd).RightS))
  ]}
 GetLineLength {nodes [
   n1 = GetLineLength!() -> (then, Start Point:Vector=~, End Point:Vector=~, Segment Number:int, SubDsegStartPointIndex:int, SubDsegEndPointIndex:int) {is=title K2Node_FunctionEntry}
   n2 = "Clamp (Integer)"(Value=@n1.SubDsegStartPointIndex, Min=0, Max=@n1.SubDsegEndPointIndex) -> (ReturnValue=0)
   n3 = Get SubdividedSplinePointLocations() -> (SubdividedSplinePointLocations:Array of Vectors)
   DrawDebugSphere!(
    execute
    Center:Vector="Get (a ref)"(Array:Array of Vectors=@n3, Dimension 1=@n2) -> (Output:"Vector (by ref)")
    Radius:float=100
    Segments:int=12
    LineColor:Linear Color Structure=(R=0.2562,G=1,B=0.2105,A=1)
    Duration:float=1
    Thickness:float=0
    DepthPriority:EDrawDebugSceneDepthPriorityGroup Enum=World
   ) -> (then)
   > DrawDebugSphere(
    Center="Get (a ref)"(Array=@n3, Dimension 1=@n1.SubDsegEndPointIndex)
    Radius=125
    Segments=12
    LineColor=(R=0.7465,G=0,B=0,A=1)
    Duration=1
    Thickness=0
    DepthPriority=World
   )
   n4 = Get SplineComponent()
   > DrawDebugSphere(
    Center=GetSplinePointAt!(self:Spline Component Object Reference=@n4, PointIndex:int=@n1.Segment Number, CoordinateSpace:ESplineCoordinateSpace Enum=World) -> (
     ReturnValue_InputKey:float=~
     ReturnValue_Position:Vector=~
     ReturnValue_ArriveTangent:Vector=~
     ReturnValue_LeaveTangent:Vector=~
     ReturnValue_Rotation:Rotator=~
     ReturnValue_Scale:Vector=~
     ReturnValue_Type:ESplinePointType Enum
    ).ReturnValue_Position
    Radius=50
    Segments=12
    LineColor=(R=0.8224,G=0,B=0.885,A=1)
    Duration=5
    Thickness=0
    DepthPriority=World
   )
   > n5 = DrawDebugSphere(
    Center=GetSplinePointAt(self=@n4, PointIndex=int + int(A=@n1.Segment Number), CoordinateSpace=World) -> (
     ReturnValue_InputKey=~
     ReturnValue_Position=~
     ReturnValue_ArriveTangent=~
     ReturnValue_LeaveTangent=~
     ReturnValue_Rotation=~
     ReturnValue_Scale=~
    ).ReturnValue_Position
    Radius=75
    Segments=12
    LineColor=(R=0.997,G=0.5352,B=0,A=1)
    Duration=5
    Thickness=0
    DepthPriority=World
   )
   n6 = Set ThsSegDist(execute=[@n5.then, @n10.Completed], ThsSegDist:double=Get AccumSegDist() -> (AccumSegDist:double)) -> (Output_Get:double)
   n7 = Vector Length(A=vector - vector(A=@n1, B=@n1.End Point))
   n8 = int + int(A=@n1.Segment Number)
   n9 = Get SplineComponent()
   > Return Node!(
    execute
    SubDSplineSegmentLength:double=@n6
    SplineSegmentLen:double=float - float(
     A=GetDistanceAlongSplineAtSplinePoint!(self:Spline Component Object Reference=@n9, PointIndex:int=@n8) -> (ReturnValue:float)
     B=GetDistanceAlongSplineAtSplinePoint(self=@n4, PointIndex=@n1.Segment Number)
    )
    Straight Dist:double="To Float (Integer)"(InInt=Truncate(A=@n7))
   )
   n10 = For Loop(execute=@n1.then, FirstIndex=@n2, LastIndex=int - int(A=@n1.SubDsegEndPointIndex))
   n11 = GetDistanceAlongSplineAtSplinePoint(self=Get ReprojectSplineComponent(), PointIndex=int + int(A=@n10))
   n12 = GetDistanceAlongSplineAtSplinePoint(self=Get ReprojectSplineComponent(), PointIndex=@n10)
   > Set AccumSegDist(AccumSegDist:double=float + float(A=float - float(A=@n11, B=@n12), B=Get AccumSegDist())) -> (Output_Get:double)
  ]}
 GetSubDSplineSegmentStartEndPointNumbers {nodes [
   n1 = GetSubDSplineSegmentStartEndPointNumbers!() -> (then, SegNum:int) {is=title K2Node_FunctionEntry}
   > Set CurrentSegmentNumber(CurrentSegmentNumber:int=@n1) -> (Output_Get:int=0)
   > Set GotSubDSplineStartPoint(GotSubDSplineStartPoint:bool=false) -> (Output_Get:bool)
   > Set GotSubDSplineEndPoint(GotSubDSplineEndPoint:bool=false) -> (Output_Get:bool)
   > n2 = For Each Loop with Break!(
    Exec:Exec
    Array:Array of Integers=Get SubdividedSplinePointSegmentNum() -> (SubdividedSplinePointSegmentNum:Array of Integers)
    Break:Exec=@n12.then
   ) -> (LoopBody:Exec, Array Element:int, Array Index:int, Completed:Exec)
   Return Node!(
    execute=@n2.Completed
    ThisSegStartPointIndex1:int=Get ThisSegStartPointIndex() -> (ThisSegStartPointIndex:int=0)
    ThisSegEndPointIndex1:int=Get ThisSegEndPointIndex() -> (ThisSegEndPointIndex:int=0)
    CurrentSegmentStartPointLoc1:Vector=Get CurrentSegmentStartPointLoc() -> (CurrentSegmentStartPointLoc:Vector)
    CurrentSegmentEndPointLoc1:Vector=Get CurrentSegmentEndPointLoc() -> (CurrentSegmentEndPointLoc:Vector)
   )
   > Set CurSubdividePointSegNum(CurSubdividePointSegNum:int=@n2) -> (Output_Get:int=0)
   > Set CurLoopIndex(CurLoopIndex:int=@n2.Array Index) -> (Output_Get:int=0)
   > n3 = Sequence!(execute) -> (then_0, then_1, then_2)
   n4 = Get GotSubDSplineStartPoint() -> (GotSubDSplineStartPoint:bool)
   n5 = Get CurrentSegmentNumber() -> (CurrentSegmentNumber:int=0)
   n6 = Get CurSubdividePointSegNum() -> (CurSubdividePointSegNum:int=0)
   > Branch(Condition=AND Boolean(A=NOT Boolean(A=@n4), B="Equal (Integer)"(A=@n6, B=@n5)))
   n7 = "Clamp (Integer)"(Value=int - int(A=Get CurLoopIndex() -> (CurLoopIndex:int=0)), Min=0, Max=10000000) -> (ReturnValue=0)
   > Set ThisSegStartPointIndex(ThisSegStartPointIndex:int=@n7) -> (Output_Get:int=0)
   > Set CurrentSegmentStartPointLoc(CurrentSegmentStartPointLoc:Vector="Get (a copy)"(Array:Array of Vectors=Get SubdividedSplinePointLocations(), Dimension 1=@n7) -> (Output:Vector)) -> (Output_Get:Vector)
   > Set GotSubDSplineStartPoint(GotSubDSplineStartPoint=true)
   n8 = Get GotSubDSplineEndPoint() -> (GotSubDSplineEndPoint:bool)
   n9 = Get CurSubdividePointSegNum() -> (CurSubdividePointSegNum=0)
   n10 = Get CurrentSegmentNumber() -> (CurrentSegmentNumber=0)
   Branch(execute=@n3.then_1, Condition=AND Boolean(A=NOT Boolean(A=@n8), B="integer > integer"(A=@n9, B=@n10)))
   n11 = int - int(A=Get CurLoopIndex() -> (CurLoopIndex=0))
   > Set ThisSegEndPointIndex(ThisSegEndPointIndex:int=@n11) -> (Output_Get:int=0)
   > Set CurrentSegmentEndPointLoc(CurrentSegmentEndPointLoc:Vector="Get (a copy)"(Array=Get SubdividedSplinePointLocations(), Dimension 1=@n11)) -> (Output_Get:Vector)
   > Set GotSubDSplineEndPoint(GotSubDSplineEndPoint=true)
   n12 = Branch(execute=@n3.then_2, Condition=AND Boolean(A=Get GotSubDSplineStartPoint(), B=Get GotSubDSplineEndPoint()))
   n13 = Get GotSubDSplineEndPoint()
   n14 = int - int(A=Length(TargetArray:Array of Integers=Get SubdividedSplinePointSegmentNum()))
   n15 = Get CurLoopIndex() -> (CurLoopIndex=0)
   Branch(execute=@n12.else, Condition=AND Boolean(A=NOT Boolean(A=@n13), B="Equal (Integer)"(A=@n15, B=@n14)))
   n16 = Get SubdividedSplinePointLocations()
   n17 = int - int(A=Length(TargetArray:Array of Vectors=@n16))
   > Set ThisSegEndPointIndex(ThisSegEndPointIndex=@n17) -> (Output_Get=0)
   > Set CurrentSegmentEndPointLoc(CurrentSegmentEndPointLoc="Get (a copy)"(Array=@n16, Dimension 1=@n17))
   > Set GotSubDSplineEndPoint(GotSubDSplineEndPoint=true)
  ]}
 GetTimeForDist {nodes [
   n1 = GetTimeForDist!() -> (then, Dist:double=~, CharSpeed:double=~, Units:E_TravelTime_SpeedUnits Enum=NewEnumerator0) {is=title K2Node_FunctionEntry}
   > K2Node_Composite_0 = Prep Vars_2!(execute, Selection:E_TravelTime_SpeedUnits Enum=@n1.Units) -> (Out:Exec) {is=title K2Node_Composite}
   n2 = Get ConversionValue() -> (ConversionValue:double)
   > Return Node!(
    execute
    Time:double=float / float(A=@n1, B=float * float(A=@n1.CharSpeed, B=@n2))
    UnitsToDisplay:String=Get UnitDisp() -> (UnitDisp:String)
   )
  ]}
 Prep Vars_2 {
  path=GetTimeForDist.K2Node_Composite_0.Prep Vars_2
  nodes [
   n1 = Inputs!() -> (execute, Selection:E_TravelTime_SpeedUnits Enum=NewEnumerator0)
   > n2 = Switch on E_TravelTime_SpeedUnits!(execute, Selection:E_TravelTime_SpeedUnits Enum=@n1) -> (NewEnumerator0:Exec, NewEnumerator1:Exec, NewEnumerator2:Exec, NewEnumerator3:Exec, NewEnumerator4:Exec) {is=title K2Node_SwitchEnum}
   n3 = MakeLiteralString(Value=MPH)
   > Set UnitDisp(UnitDisp:String=@n3) -> (Output_Get:String)
   > n4 = Set ConversionValue(ConversionValue:double=Find(
     TargetMap:"Map of Strings to Float (double-precision)s"=Get SpeedConversionFactors() -> (SpeedConversionFactors:"Map of Strings to Float (double-precision)s")
     Key=@n3
    ) -> (Value:double)) -> (Output_Get:double)
   Outputs!(Out:Exec=[@n10.then, @n8.then, @n6.then, @n4.then])
   n5 = MakeLiteralString(Value=KPH)
   Set UnitDisp(execute=@n2.NewEnumerator3, UnitDisp=@n5)
   > n6 = Set ConversionValue(ConversionValue=Find(TargetMap=Get SpeedConversionFactors(), Key=@n5))
   n7 = MakeLiteralString(Value=cm/s)
   Set UnitDisp(execute=@n2.NewEnumerator2, UnitDisp=@n7)
   > n8 = Set ConversionValue(ConversionValue=Find(TargetMap=Get SpeedConversionFactors(), Key=@n7))
   n9 = MakeLiteralString(Value=m/s)
   Set UnitDisp(execute=@n2.NewEnumerator1, UnitDisp=@n9)
   > n10 = Set ConversionValue(ConversionValue=Find(TargetMap=Get SpeedConversionFactors(), Key=@n9))
  ]
 }
 GizmoTransformChanged {nodes [
   n1 = GizmoTransformChanged!() -> (then, Gizmo Ident:String, New Transform:Transform) {is=title K2Node_FunctionEntry}
   > n2 = GetGizmoDigit!(execute, self, GizID:String=@n1) -> (then, Digit:int=0) {is=title K2Node_CallFunction}
   > n3 = Set CurGizmoDigit(CurGizmoDigit:int=@n2) -> (Output_Get:int=0)
   > n4 = Set GizmoCurLocation(GizmoCurLocation:Vector=BreakTransform(InTransform=@n1.New Transform)) -> (Output_Get:Vector)
   > Set Array Elem!(
    execute
    TargetArray:Array of Vectors=Get AllSplinePoints() -> (AllSplinePoints:Array of Vectors)
    Index:int=@n3
    Item:"Vector (by ref)"=@n4
    bSizeToFit:bool=false
   ) -> (then)
   > RefreshAll(ImmediateUpdate=false)
  ]}
 GizmoTransformStateChange {nodes [
   n1 = GizmoTransformStateChange!() -> (then, GizIdentifier:String, NewTransform:Transform, ChangeType:EScriptableToolGizmoStateChangeType Enum=BeginTransform) {is=title K2Node_FunctionEntry}
   > n2 = GetGizmoTransform!(execute, self:Scriptable Interactive Tool Object Reference, Identifier:String=@n1) -> (then, ReturnValue:Transform)
   > Set GizmoTransformStateChangeSnapshot(GizmoTransformStateChangeSnapshot:Transform=@n2) -> (Output_Get:Transform)
   > n3 = Switch on EScriptableToolGizmoStateChangeType!(execute, Selection:EScriptableToolGizmoStateChangeType Enum=@n1.ChangeType) -> (BeginTransform:Exec, EndTransform:Exec, UndoRedo:Exec) {is=title K2Node_SwitchEnum}
   > n4 = GetDynamicMesh!(execute, self:Base Dynamic Mesh Component Object Reference=Get Dynamic Mesh Component()) -> (then, ReturnValue:Dynamic Mesh Object Reference)
   > n5 = Set DynMesh(DynMesh:Dynamic Mesh Object Reference=@n4) -> (Output_Get:Dynamic Mesh Object Reference)
   > n6 = BeginTrackedMeshChange!(execute, TargetMesh:Dynamic Mesh Object Reference=@n5) -> (then, ChangeTracker:Dynamic Mesh Change Tracker Structure, ReturnValue:Dynamic Mesh Object Reference)
   > Set DynMeshChangeTracker(DynMeshChangeTracker:Dynamic Mesh Change Tracker Structure=@n6) -> (Output_Get:Dynamic Mesh Change Tracker Structure)
   > GizmoTransformChanged(Gizmo Ident=Get GizIdentifier() -> (GizIdentifier:String), New Transform=Get NewTransform() -> (NewTransform:Transform)) {is=title K2Node_CallFunction}
   > EmitTrackedMeshChange!(
    execute
    TargetMesh:Dynamic Mesh Object Reference=Get DynMesh() -> (DynMesh:Dynamic Mesh Object Reference)
    ChangeTracker:"Dynamic Mesh Change Tracker Structure (by ref)"=Get DynMeshChangeTracker() -> (DynMeshChangeTracker:Dynamic Mesh Change Tracker Structure)
   ) -> (then, ReturnValue:Dynamic Mesh Object Reference)
   RefreshAll(execute=@n3.EndTransform, ImmediateUpdate=true)
   GizmoTransformChanged(execute=@n3.UndoRedo, Gizmo Ident=Get GizIdentifier(), New Transform=Get NewTransform())
   > RefreshAll(ImmediateUpdate=true)
  ]}
 GizmoVizControl {nodes [
   n1 = GizmoVizControl!() -> (then, Visible:bool=true) {is=title K2Node_FunctionEntry}
   > Set GizmosVisible(GizmosVisible:bool=@n1) -> (Output_Get:bool)
   > n2 = For Each Loop(Array=Get AllGizmoNames())
   > SetGizmoVisible(Identifier=@n2, bVisible=@n1)
  ]}
 IfHitByClick {nodes [
   n1 = IfHitByClick!() -> (
    then
    ClickPos_WorldRay:Ray Structure
    ClickPos_bHas2D:bool=~
    ClickPos_ScreenPosition:Vector 2D Structure
    MouseButton:EScriptableToolMouseButton Enum=LeftButton
   ) {is=title K2Node_FunctionEntry}
   > Set MouseClickButton(MouseClickButton:EScriptableToolMouseButton Enum=@n1.MouseButton) -> (Output_Get:EScriptableToolMouseButton Enum=LeftButton)
   n2 = GetRayStartEnd!(Ray:Ray Structure=@n1, StartDistance:double=0, EndDistance:double=0) -> (StartPoint:Vector, EndPoint:Vector)
   > n3 = Line Trace By Channel!(
    execute
    Start:Vector=@n2
    End:Vector=@n2.EndPoint
    TraceChannel:ETraceTypeQuery Enum
    bTraceComplex:bool=true
    ActorsToIgnore:Array of Actor Object References=Get All Measure Tool Spline Actors EUA()
    DrawDebugType:EDrawDebugTrace Enum
    bIgnoreSelf:bool
    TraceColor:Linear Color Structure
    TraceHitColor:Linear Color Structure
    DrawTime:float
   ) -> (then, OutHit:Hit Result Structure, ReturnValue:bool)
   > Set SingleClickHit(SingleClickHit:Hit Result Structure=@n3) -> (Output_Get:Hit Result Structure)
   > Return Node!(
    execute
    ReturnValue_bHit:bool=@n3.ReturnValue
    ReturnValue_HitDepth:double=340282346638528859811704183484516925440
    ReturnValue_HitNormal:Vector="0,0,1"
    ReturnValue_bHasHitNormal:bool=False
    ReturnValue_HitIdentifier:int=0
    ReturnValue_HitObject:Object Reference
   )
  ]}
 MakeBillBoardTextToRender {nodes [
   n1 = MakeBillBoardTextToRender!() -> (then, CurLineLen:double=~, CurLineRunTime:double=~, SegmentNumber:int, CurLineLenSubD:double=~, CurLinLenStraight:double=~) {is=title K2Node_FunctionEntry}
   n2 = Get Settings()
   n3 = Enum to String!(Enumerator:E_TravelTime_DistUnits Enum=Get Distance Units!(self:BP STPS Travel Time Object Reference=@n2) -> (Distance Units:E_TravelTime_DistUnits Enum=NewEnumerator5)) -> (ReturnValue:String)
   n4 = Enum to String(Enumerator:E_TravelTime_TimeUnits Enum=Get Time Units!(self:BP STPS Travel Time Object Reference=@n2) -> (Time Units:E_TravelTime_TimeUnits Enum=NewEnumerator5))
   > n5 = GetConvertedUnits!(
    execute
    Distance in cm:double=@n1.CurLineLenSubD
    Distance Units:String=@n3
    Time in sec:double=@n1.CurLineRunTime
    Time Units:String=@n4
    Grid Units:double="To Float (Integer)"(InInt=Get Grid Snap Size!(self:BP STPS Travel Time Object Reference=@n2) -> (Grid Snap Size:int=0))
   ) -> (then, Converted distance:double, Distance Units Final:String, Converted time:double, Time Units Final:String)
   n6 = MakeLiteralInt!(Value:int=0) -> (ReturnValue:int=0)
   n7 = Select(Option 0:int=3, Option 1:int=@n6, Index="Equal, Case Insensitive (String)"(A=@n3, B=cm)) -> (ReturnValue:int)
   n8 = Select(Option 0=3, Option 1=@n6, Index="Equal Exactly (String)"(A=@n4, B=sec))
   > Return Node!(
    execute
    Result:Text=Format Text!(
     Format="LOCTEXT(\"Segment {s}\\r\\n{ll} {distunits}\\r\\n{runtime} {timeunits}\")"
     s=int + int(A=@n1.SegmentNumber)
     ll:Text="To Text (Float)"(Value=@n5, MinimumIntegralDigits=1, MaximumIntegralDigits=324, MinimumFractionalDigits=0, MaximumFractionalDigits=@n7)
     distunits:String=@n5.Distance Units Final
     runtime:Text="To Text (Float)"(Value=@n5.Converted time, MinimumIntegralDigits=1, MaximumIntegralDigits=324, MinimumFractionalDigits=0, MaximumFractionalDigits=@n8)
     timeunits:String=@n5.Time Units Final
    ) -> (Result)
   )
  ]}
 MakeSubobjectTextRender {nodes [
   MakeSubobjectTextRender!() -> (then, TheTransform:Transform) {is=title K2Node_FunctionEntry}
   > n1 = Get Scene Root Handle() {is=title K2Node_CallFunction}
   > n2 = AddNewSubobject(
    self=Get SubobjectDataSubsystem()
    Params=Make Add New Subobject Params(ParentHandle=@n1, NewClass=/Script/Engine.TextRenderComponent, bSkipMarkBlueprintModified=False, bConformTransformToParent=True)
   )
   > n3 = Cast To TextRenderComponent!(execute, Object=GetAssociatedObject(Data=GetData(DataHandle=@n2.ReturnValue))) -> (then, CastFailed, AsText Render Component:Text Render Component Object Reference) {is=title K2Node_DynamicCast}
   > Set World Transform!(
    execute
    self:Scene Component Object Reference=@n3
    NewTransform:"Transform (by ref)"=Get TheTransform() -> (TheTransform:Transform)
    bSweep:bool=false
    bTeleport:bool=false
   ) -> (then, SweepHitResult:Hit Result Structure) {is=title K2Node_CallFunction}
   > SetComponentEditorOnly(Object=@n3)
   > Return Node!(execute, Text Render Component:Text Render Component Object Reference=@n3)
  ]}
 On Units Changed {nodes [
   On Units Changed!() -> (then, PropertySet:Scriptable Interactive Tool Property Set Object Reference, PropertyName:String) {is=title K2Node_FunctionEntry}
   > RefreshAll(ImmediateUpdate=true)
  ]}
 OnBeginHoverSeqHitTest {nodes [
   OnBeginHoverSeqHitTest!() -> (then, CurrentPos:Input Device Ray Structure, Modifiers:Scriptable Tool Modifier States Structure) {is=title K2Node_FunctionEntry}
   > SetFocusInViewport(self:Scriptable Interactive Tool Object Reference)
   > Return Node!(execute, ReturnValue:Input Ray Hit Structure)
  ]}
 OnBillboardSizeChanged {nodes [
   n1 = OnBillboardSizeChanged!() -> (then, PropertySet:Scriptable Interactive Tool Property Set Object Reference, PropertyName:String, NewValue:double=~) {is=title K2Node_FunctionEntry}
   > n2 = For Each Loop(Array:Array of Material Instance Dynamic Object References=Get CurTextRenderDynMaterials() -> (CurTextRenderDynMaterials:Array of Material Instance Dynamic Object References)) -> (Array Element:Material Instance Dynamic Object Reference)
   n3 = For Each Loop(Exec=@n2.Completed, Array=Get CurTextBGDynMaterials() -> (CurTextBGDynMaterials:Array of Material Instance Dynamic Object References))
   > SetScalarParameterValue(self=@n3, ParameterName=camdistscale, Value=float * float(A=@n1.NewValue, B=Get BillboardScaleFactor() -> (BillboardScaleFactor:double)))
   SetScalarParameterValue(execute=@n2.LoopBody, self=@n2, ParameterName=camdistscale, Value=float * float(A=@n1.NewValue, B=Get BillboardScaleFactor()))
  ]}
 OnCleanAllClicked {nodes [
   OnCleanAllClicked() {is=title K2Node_FunctionEntry}
   > n1 = GetAllActorsOfClass(ActorClass=&BP_EUA_TravelTime_SplineActor_C)
   n2 = int - int(A=Length(TargetArray:Array of BP EUA Travel Time Spline Actor Object References=@n1))
   > n3 = Branch(Condition="integer > integer"(A=@n2))
   n4 = "To Text (String)"(InString=Get ClearSplinesWarnIniName() -> (ClearSplinesWarnIniName:String))
   n5 = Get ToolName(None=Self-Reference()) -> (ToolName:Text)
   n6 = "Not Equal (Integer)"(A=@n2, B=1)
   n7 = Get ClearSplinesWarnMessage_p() -> (ClearSplinesWarnMessage_p:Text)
   n8 = Get ClearSplinesWarnMessage_s() -> (ClearSplinesWarnMessage_s:Text)
   > n9 = Show Suppressable Warning Dialog!(
    execute
    Title:"Text (by ref)"=Get ClearSplinesWarnTitle() -> (ClearSplinesWarnTitle:Text)
    Message:"Text (by ref)"=Format Text!(Format=SelectText(A=@n7, B=@n8, bPickA=@n6), 0:int=@n2) -> (Result)
    InIniSettingName:String="To String (Text)"(InText=Format Text(Format=@n4, 0:Text=@n5))
    InIniSettingFileNameOverride:String
    bDefaultValue:bool=true
   ) -> (then, ReturnValue:bool)
   > Branch(Condition=@n9)
   > BeginTransaction(Context=Get TransactionContext() -> (TransactionContext:String), Description="LOCTEXT(\"Delete Actor\")")
   > n10 = For Each Loop(Array:Array of BP EUA Travel Time Spline Actor Object References=@n1) -> (Array Element:BP EUA Travel Time Spline Actor Object Reference)
   EndTransaction(execute=@n10.Completed)
   n11 = Branch(execute=@n10.LoopBody, Condition="Equal (Object)"(A=@n10, B=Get SplineActorEUA()))
   Destroy Actor(execute=@n11.else, self=@n10)
   DisplayUserWarningMessage(execute=@n3.else, Message=Get ClearSplinesNoneFoundWarnMessage() -> (ClearSplinesNoneFoundWarnMessage:Text))
  ]}
 OnCleanCurSegmentsClicked {nodes [
   OnCleanCurSegmentsClicked() {is=title K2Node_FunctionEntry}
   n1 = Get AllSplinePoints()
   > n2 = Branch(Condition="integer >= integer"(A=Length(TargetArray:Array of Vectors=@n1), B=1))
   > BeginTransaction(Context=Get TransactionContext(), Description="LOCTEXT(\"Delete Actor\")")
   > TransactObject(Object=Get SplineActorEUA())
   > TransactObject(Object=Self-Reference())
   > Clear(TargetArray:Array of Vectors=Get AllSplinePoints())
   > Clear(TargetArray=Get SplinePointPositions(self=Get SplineActorEUA()))
   > Set TotalTime(TotalTime=0)
   > Set TotalDist(TotalDist=0)
   > Destroy Actor(self=Get SplineActorEUA())
   > n3 = AddSplineActor()
   > Set SplineActorEUA(SplineActorEUA=@n3)
   > RefreshAll(ImmediateUpdate=true)
   > EndTransaction()
   DisplayUserWarningMessage(execute=@n2.else, Message=Get ClearSegmentsNoneFoundWarnMessage() -> (ClearSegmentsNoneFoundWarnMessage:Text))
  ]}
 OnColorPropertyChanged {nodes [
   OnColorPropertyChanged!() -> (then, PropertySet:Scriptable Interactive Tool Property Set Object Reference, PropertyName:String) {is=title K2Node_FunctionEntry}
   > RefreshAll(ImmediateUpdate=true)
  ]}
 OnConfigBoolsChanged {nodes [
   OnConfigBoolsChanged!() -> (then, PropertySet:Scriptable Interactive Tool Property Set Object Reference, PropertyName:String, bNewValue:bool=~) {is=title K2Node_FunctionEntry}
   > SetSplineCompViz()
   > RefreshAll(ImmediateUpdate=true)
  ]}
 OnDeleteHotkeyDown {nodes [
   OnDeleteHotkeyDown!() -> (then, Key:Key Structure=None, Modifiers:Scriptable Tool Modifier States Structure) {is=title K2Node_FunctionEntry}
   > Set IsDeleteKeyDown(IsDeleteKeyDown:bool=true) -> (Output_Get:bool)
  ]}
 OnDeleteHotkeyUp {nodes [
   OnDeleteHotkeyUp!() -> (then, Key:Key Structure=None, Modifiers:Scriptable Tool Modifier States Structure) {is=title K2Node_FunctionEntry}
   > Set IsDeleteKeyDown(IsDeleteKeyDown=false)
  ]}
 OnEditHotkeyPressed {nodes [
   OnEditHotkeyPressed!() -> (then, Key:Key Structure=None, Modifiers:Scriptable Tool Modifier States Structure) {is=title K2Node_FunctionEntry}
   > Set isEditHotkeyDown(isEditHotkeyDown:bool=true) -> (Output_Get:bool)
   > AddPointGizmos!(execute, self, Vizible:bool=true) -> (then) {is=title K2Node_CallFunction}
  ]}
 OnEditHotkeyReleased {nodes [
   OnEditHotkeyReleased!() -> (then, Key:Key Structure=None, Modifiers:Scriptable Tool Modifier States Structure) {is=title K2Node_FunctionEntry}
   > Set isEditHotkeyDown(isEditHotkeyDown=false)
   > GizmoVizControl!(execute, self, Visible:bool=false) -> (then) {is=title K2Node_CallFunction}
   > DestroyPointGizmos() {is=title K2Node_CallFunction}
   > RefreshAll(ImmediateUpdate=true)
  ]}
 OnFloatPropertyChanged {nodes [
   OnFloatPropertyChanged!() -> (then, PropertySet:Scriptable Interactive Tool Property Set Object Reference, PropertyName:String, NewValue:double=~) {is=title K2Node_FunctionEntry}
   > RefreshAll(ImmediateUpdate=false)
  ]}
 OnFortSpeedPresetsPropertyChanged {nodes [
   OnFortSpeedPresetsPropertyChanged!() -> (then, PropertySet:Scriptable Interactive Tool Property Set Object Reference, PropertyName:String) {is=title K2Node_FunctionEntry}
   n1 = Get Settings()
   n2 = Find(
    TargetMap=Get Fort Speed Presets Map() -> (Fort Speed Presets Map:"Map of Strings to Float (double-precision)s")
    Key=Enum to String(Enumerator:E_TravelTime_SpeedPresets Enum=Get Movement Presets(self=@n1) -> (Movement Presets=NewEnumerator0))
   )
   > Branch(Condition=@n2.ReturnValue)
   > Set PreviousMovementPreset(PreviousMovementPreset=Get Movement Presets(self=Get Settings()) -> (Movement Presets=NewEnumerator0)) -> (Output_Get=NewEnumerator0)
   n3 = Get Settings()
   > Set Previous Movement Speed(Previous Movement Speed=Get Movement Speed(self=@n3))
   > Set Movement Speed!(execute, Movement Speed:double=@n2, self:BP STPS Travel Time Object Reference=@n3) -> (then, Output_Get:double)
  ]}
 OnGridSnappingChanged {nodes [
   OnGridSnappingChanged!() -> (then, PropertySet:Scriptable Interactive Tool Property Set Object Reference, PropertyName:String) {is=title K2Node_FunctionEntry}
   n1 = Get Settings()
   > SetPropertyAsReadOnly!(
    execute
    self:Scriptable Interactive Tool Property Set Object Reference=@n1
    PropertyName:String=Grid Snap Size
    bReadOnly:bool=NOT Boolean(A=Get Grid Snapping(self=@n1))
   ) -> (then, ReturnValue:Scriptable Interactive Tool Property Set Object Reference)
  ]}
 OnHitByClick {nodes [
   OnHitByClick!() -> (
    then
    ClickPos:Input Device Ray Structure
    Modifiers:Scriptable Tool Modifier States Structure
    MouseButton:EScriptableToolMouseButton Enum=LeftButton
   ) {is=title K2Node_FunctionEntry}
   > n1 = Branch(Condition=Get isSplitHotkeyDown() -> (isSplitHotkeyDown:bool))
   > OnHitByClickInsertSegment()
   n2 = Branch(execute=@n1.else, Condition=Get IsDeleteKeyDown() -> (IsDeleteKeyDown:bool))
   > OnHitByClickDeleteSegment!(execute, self, PointToRemove:int=-1) -> (then)
   OnHitByClickMakeSpline!(execute=[@n2.else, @n3.else], self, PrependPoint:bool=Get isPrependKeyDown() -> (isPrependKeyDown:bool)) -> (then)
   n3 = Branch()
   Get isEditHotkeyDown()
  ]}
 OnHitByClickDeleteSegment {nodes [
   n1 = OnHitByClickDeleteSegment!() -> (then, PointToRemove:int=-1) {is=title K2Node_FunctionEntry}
   > BeginTransaction(Context=Get TransactionContext(), Description=Get DeleteSegmentTransactionDescription() -> (DeleteSegmentTransactionDescription:Text))
   > TransactObject(Object=Get SplineComponent())
   > TransactObject(Object=Self-Reference())
   > n2 = GetDynamicMesh(self=Get Dynamic Mesh Component())
   > n3 = BeginTrackedMeshChange(TargetMesh=@n2)
   n4 = FindInputKeyClosestToWorldLocation!(
    self:Spline Component Object Reference=Get SplineComponent()
    WorldLocation:"Vector (by ref)"=Get!() -> (
     SingleClickHit_bBlockingHit:bool
     SingleClickHit_bInitialOverlap:bool
     SingleClickHit_Time:float
     SingleClickHit_Distance:float
     SingleClickHit_Location:Vector
     SingleClickHit_ImpactPoint:Vector
     SingleClickHit_Normal:Vector
     SingleClickHit_ImpactNormal:Vector
     SingleClickHit_PhysMat:Physical Material Object Reference
     SingleClickHit_HitActor:Actor Object Reference
     SingleClickHit_HitComponent:Primitive Component Object Reference
     SingleClickHit_HitBoneName:Name=None
     SingleClickHit_BoneName:Name=None
     SingleClickHit_HitItem:int=0
     SingleClickHit_ElementIndex:int=0
     SingleClickHit_FaceIndex:int=0
     SingleClickHit_TraceStart:Vector
     SingleClickHit_TraceEnd:Vector
    ).SingleClickHit_Location
   ) -> (ReturnValue:float)
   n5 = Round(A=@n4) {is=title K2Node_CallFunction}
   > Branch(Condition="integer > integer"(A=@n5, B=-1))
   > Remove Index!(execute, TargetArray:Array of Vectors=Get AllSplinePoints(), IndexToRemove:int=@n5) -> (then)
   > EmitTrackedMeshChange(TargetMesh=@n3.ReturnValue, ChangeTracker=@n3)
   > RefreshAll(ImmediateUpdate=true)
   > EndTransaction()
   SelectInt(A=@n1, B=Ceil(A=@n4) {is=title K2Node_CallFunction}, bPickA="integer >= integer"(A=@n1))
  ]}
 OnHitByClickInsertSegment {nodes [
   OnHitByClickInsertSegment() {is=title K2Node_FunctionEntry}
   > BeginTransaction(Context=Get TransactionContext(), Description="LOCTEXT(\"Creator Actor\")")
   > TransactObject(Object=Get SplineComponent())
   > TransactObject(Object=Self-Reference())
   > n1 = SnapVectorToGrid!(
    execute
    VectorToSnap:Vector=BreakHitResult!(Hit:"Hit Result Structure (by ref)"=Get SingleClickHit() -> (SingleClickHit:Hit Result Structure)) -> (
     bBlockingHit:bool
     bInitialOverlap:bool
     Time:float
     Distance:float
     Location:Vector
     ImpactPoint:Vector
     Normal:Vector
     ImpactNormal:Vector
     PhysMat:Physical Material Object Reference
     HitActor:Actor Object Reference
     HitComponent:Primitive Component Object Reference
     HitBoneName:Name
     BoneName:Name
     HitItem:int
     ElementIndex:int
     FaceIndex:int
     TraceStart:Vector
     TraceEnd:Vector
    ).Location
    GridSize:int=Get Grid Snap Size(self=Get Settings()) -> (Grid Snap Size=0)
    Snap On:bool=Get Grid Snapping(self=Get Settings())
    Ignore Up:bool=false
   ) -> (then, OutputVector:Vector)
   > n2 = GetDynamicMesh(self=Get Dynamic Mesh Component())
   > n3 = Set DynMesh(DynMesh=@n2)
   > n4 = BeginTrackedMeshChange(TargetMesh=@n3)
   > Set DynMeshChangeTracker(DynMeshChangeTracker=@n4)
   n5 = FindInputKeyClosestToWorldLocation(self=Get SplineComponent(), WorldLocation=@n1)
   > Insert!(
    execute
    TargetArray:Array of Vectors=Get AllSplinePoints()
    NewItem:"Vector (by ref)"=GetLocationAtSplineInputKey!(self:Spline Component Object Reference=Get SplineComponent(), InKey:float=@n5, CoordinateSpace:ESplineCoordinateSpace Enum=World) -> (ReturnValue:Vector)
    Index:int=int + int(A=Floor(A=@n5) {is=title K2Node_CallFunction})
   ) -> (then) {is=title K2Node_CallArrayFunction}
   > RefreshAll(ImmediateUpdate=true)
   > EmitTrackedMeshChange(TargetMesh=Get DynMesh(), ChangeTracker=Get DynMeshChangeTracker())
   > EndTransaction()
  ]}
 OnHitByClickMakeSpline {nodes [
   OnHitByClickMakeSpline!() -> (then, PrependPoint:bool=~) {is=title K2Node_FunctionEntry}
   > BeginTransaction(Context=Get TransactionContext(), Description="LOCTEXT(\"Add Actor\")")
   > n1 = SnapVectorToGrid(
    VectorToSnap=BreakHitResult(Hit=Get SingleClickHit()).Location
    GridSize=Get Grid Snap Size(self=Get Settings()) -> (Grid Snap Size=0)
    Snap On=Get Grid Snapping(self=Get Settings())
    Ignore Up=false
   )
   > TransactObject(Object=Get SplineComponent())
   > TransactObject(Object=Get ReprojectSplineComponent())
   > TransactObject(Object=Self-Reference())
   > n2 = GetDynamicMesh(self=Get Dynamic Mesh Component())
   > n3 = BeginTrackedMeshChange(TargetMesh=@n2)
   > Set DynMeshChangerTracker(DynMeshChangerTracker:Dynamic Mesh Change Tracker Structure=@n3) -> (Output_Get:Dynamic Mesh Change Tracker Structure)
   > Set DynMesh(DynMesh=@n3.ReturnValue)
   > n4 = Branch(Condition=Get PrependPoint() -> (PrependPoint:bool))
   n5 = vector + vector(A=@n1, B=Get Spline Hit Z Tweak() -> (Spline Hit Z Tweak:Vector))
   n6 = Get AllSplinePoints()
   > Branch(Condition="integer < integer"(A=Find Item!(TargetArray:Array of Vectors=@n6, ItemToFind:"Vector (by ref)"=@n5) -> (ReturnValue)))
   > n7 = Insert(TargetArray=Get AllSplinePoints(), NewItem=@n5, Index=0)
   RefreshAll(execute=[@n8.then, @n7.then], ImmediateUpdate=true)
   > EmitTrackedMeshChange(TargetMesh=Get DynMesh(), ChangeTracker=Get DynMeshChangerTracker() -> (DynMeshChangerTracker:Dynamic Mesh Change Tracker Structure))
   > EndTransaction()
   n8 = Add Unique(execute=@n4.else, TargetArray:Array of Vectors=Get AllSplinePoints(), NewItem:"Vector (by ref)"=@n5)
  ]}
 OnHotkeyInterrupted {nodes [
   OnHotkeyInterrupted() {is=title K2Node_FunctionEntry}
   > n1 = Sequence!(execute) -> (then_0, then_1, then_2, then_3:Exec)
   > Branch(Condition=Get isSplitHotkeyDown())
   > OnSplitHotkeyUp!(execute, self, Key:Key Structure=None, Modifiers:Scriptable Tool Modifier States Structure) -> (then)
   Branch(execute=@n1.then_1, Condition=Get IsDeleteKeyDown())
   > OnDeleteHotkeyUp!(execute, self, Key:Key Structure=None, Modifiers:Scriptable Tool Modifier States Structure) -> (then) {is=title K2Node_CallFunction}
   Branch(execute=@n1.then_2, Condition=Get isPrependKeyDown())
   > OnPrependHotkeyUp!(execute, self, Key:Key Structure=None, Modifiers:Scriptable Tool Modifier States Structure) -> (then)
   Branch(execute=@n1.then_3, Condition=Get isEditHotkeyDown())
   > OnEditHotkeyReleased!(execute, self, Key:Key Structure=None, Modifiers:Scriptable Tool Modifier States Structure) -> (then) {is=title K2Node_CallFunction}
  ]}
 OnMovementSpeedChanged {nodes [
   OnMovementSpeedChanged!() -> (then, PropertySet:Scriptable Interactive Tool Property Set Object Reference, PropertyName:String, NewValue:double=~) {is=title K2Node_FunctionEntry}
   > n1 = Sequence!(execute) -> (then_0, then_1)
   > Set CurrentPreset(CurrentPreset:E_TravelTime_SpeedPresets Enum=Get Movement Presets(self=Get Settings()) -> (Movement Presets=NewEnumerator0)) -> (Output_Get:E_TravelTime_SpeedPresets Enum=NewEnumerator0)
   > n2 = Branch(Condition="Equal (Enum)"(
     A:E_TravelTime_SpeedPresets Enum=Get CurrentPreset() -> (CurrentPreset:E_TravelTime_SpeedPresets Enum=NewEnumerator0)
     B:E_TravelTime_SpeedPresets Enum=NewEnumerator22
    ))
   > Set PreviousMovementPreset(PreviousMovementPreset=Get Movement Presets(self=Get Settings()) -> (Movement Presets=NewEnumerator0)) -> (Output_Get=NewEnumerator0)
   > Set Previous Movement Speed(Previous Movement Speed=Get Movement Speed(self=Get Settings()))
   n3 = Branch(
    execute=@n2.else
    Condition="Equal (Enum)"(
     A=Get PreviousMovementPreset() -> (PreviousMovementPreset:E_TravelTime_SpeedPresets Enum=NewEnumerator0)
     B=Get CurrentPreset() -> (CurrentPreset=NewEnumerator0)
    )
   )
   n4 = Find(TargetMap=Get Fort Speed Presets Map(), Key=Enum to String(Enumerator=Get CurrentPreset() -> (CurrentPreset=NewEnumerator0)))
   > Branch(Condition=@n4.ReturnValue)
   n5 = Get Settings()
   > n6 = Branch(Condition="Equal (Float)"(A=@n4, B=Get Movement Speed(self=@n5)))
   > Set PreviousMovementPreset(PreviousMovementPreset=Get Movement Presets(self=Get Settings()) -> (Movement Presets=NewEnumerator0)) -> (Output_Get=NewEnumerator0)
   > Set Previous Movement Speed(Previous Movement Speed=Get Movement Speed(self=Get Settings()))
   Set PreviousMovementPreset(execute=@n6.else, PreviousMovementPreset=Get Movement Presets(self=Get Settings()) -> (Movement Presets=NewEnumerator0)) -> (Output_Get=NewEnumerator0)
   > Set Movement Presets!(execute, Movement Presets:E_TravelTime_SpeedPresets Enum=NewEnumerator22, self:BP STPS Travel Time Object Reference=Get Settings()) -> (then, Output_Get:E_TravelTime_SpeedPresets Enum=NewEnumerator0)
   > Set Previous Movement Speed(Previous Movement Speed=Get Movement Speed(self=Get Settings()))
   Set PreviousMovementPreset(execute=@n3.else, PreviousMovementPreset=Get Movement Presets(self=Get Settings()) -> (Movement Presets=NewEnumerator0)) -> (Output_Get=NewEnumerator0)
   > Set Previous Movement Speed(Previous Movement Speed=Get Movement Speed(self=Get Settings()))
   RefreshAll(execute=@n1.then_1, ImmediateUpdate=true)
  ]}
 OnMovementSpeedMultiplierChanged {nodes [
   OnMovementSpeedMultiplierChanged!() -> (then, PropertySet:Scriptable Interactive Tool Property Set Object Reference, PropertyName:String, NewValue:double=~) {is=title K2Node_FunctionEntry}
   > RefreshAll(ImmediateUpdate=true)
  ]}
 OnPrependHotkeyDown {nodes [
   OnPrependHotkeyDown!() -> (then, Key:Key Structure=None, Modifiers:Scriptable Tool Modifier States Structure) {is=title K2Node_FunctionEntry}
   > Set isPrependKeyDown(isPrependKeyDown:bool=true) -> (Output_Get:bool)
  ]}
 OnPrependHotkeyUp {nodes [
   OnPrependHotkeyUp!() -> (then, Key:Key Structure=None, Modifiers:Scriptable Tool Modifier States Structure) {is=title K2Node_FunctionEntry}
   > Set isPrependKeyDown(isPrependKeyDown=false)
  ]}
 OnScriptShudown {nodes [
   OnScriptShudown() {is=title K2Node_FunctionEntry}
   > n1 = Sequence!(execute) -> (then_0, then_1, then_2:Exec)
   > Is Valid(InputObject=Get SplineActorEUA())
   n2 = Get AllSplinePoints()
   > n3 = Branch(Condition="integer < integer"(A=Length(TargetArray=@n2), B=2))
   > Destroy Actor(self=Get SplineActorEUA())
   Sequence!(execute=@n3.else) -> (then_0, then_1, then_2, then_3:Exec, then_4:Exec, then_5:Exec, then_6:Exec)
   > RefreshAll(ImmediateUpdate=true)
   n4 = Get Readout Box()
   n5 = Get Shared State Settings()
   Set ReadoutOpenOnExit!(
    execute=@n1.then_1
    ReadoutOpenOnExit:bool=IsChecked(self=Get ExpandButton!(self:BP UW Expandable Box Object Reference=@n4) -> (ExpandButton:Check Box Object Reference))
    self:BP STPS Shared Widget State Object Reference=@n5
   ) -> (then, Output_Get:bool)
   n6 = Get Shortcut Box()
   > Set ShortcutsOpenOnExit!(execute, ShortcutsOpenOnExit:bool=IsChecked(self=Get ExpandButton(self=@n6)), self:BP STPS Shared Widget State Object Reference=@n5) -> (then, Output_Get:bool)
   SavePropertySetSettings(execute=@n1.then_2, PropertySet=Get Shared State Settings(), SaveKey=Get PropSetSaveRestoreKey())
   > SavePropertySetSettings(PropertySet=Get Settings(), SaveKey=Get PropSetSaveRestoreKey())
   n7 = Switch on E_TravelTime_ReprojectOptions!(
    execute
    Selection:E_TravelTime_ReprojectOptions Enum=Get Follow Surface!(self:BP STPS Travel Time Object Reference=Get Settings()) -> (Follow Surface:E_TravelTime_ReprojectOptions Enum=NewEnumerator0)
   ) -> (NewEnumerator0:Exec, NewEnumerator1:Exec, NewEnumerator2:Exec) {is=title K2Node_SwitchEnum}
   > ClearSplinePoints(self=Get ReprojectSplineComponent(), bUpdateSpline=true)
   ClearSplinePoints(execute=[@n7.NewEnumerator1, @n7.NewEnumerator2], self=Get SplineComponent(), bUpdateSpline=true)
   SaveSplineGeoToAsset()
  ]}
 OnSplitHotkeyDown {nodes [
   OnSplitHotkeyDown!() -> (then, Key:Key Structure=None, Modifiers:Scriptable Tool Modifier States Structure) {is=title K2Node_FunctionEntry}
   > Set isSplitHotkeyDown(isSplitHotkeyDown:bool=true) -> (Output_Get:bool)
  ]}
 OnSplitHotkeyUp {nodes [
   OnSplitHotkeyUp!() -> (then, Key:Key Structure=None, Modifiers:Scriptable Tool Modifier States Structure) {is=title K2Node_FunctionEntry}
   > Set isSplitHotkeyDown(isSplitHotkeyDown=false)
  ]}
 PostSetupInitUserWidget {nodes [
   n1 = PostSetupInitUserWidget!() -> (
    then
    ReadoutBox:BP UW Expandable Box Object Reference
    ShortcutBox:BP UW Expandable Box Object Reference
    Shared Widget State:BP STPS Shared Widget State Object Reference
   ) {is=title K2Node_FunctionEntry}
   n2 = Get ReadoutOpenOnExit(self=@n1.Shared Widget State)
   > SetBoxState!(execute, self:BP UW Expandable Box Object Reference=@n1, Checked:bool=@n2) -> (then)
   > SetCheckedState!(
    execute
    self:Check Box Object Reference=Get ExpandButton(self=@n1)
    InCheckedState:ECheckBoxState Enum=Select(Option 0:ECheckBoxState Enum=Unchecked, Option 1:ECheckBoxState Enum=Checked, Index=@n2) -> (ReturnValue:ECheckBoxState Enum)
   ) -> (then)
   n3 = Get ShortcutsOpenOnExit(self=@n1.Shared Widget State)
   > SetCheckedState(self=Get ExpandButton(self=@n1.ShortcutBox), InCheckedState=Select(Option 0=Unchecked, Option 1=Checked, Index=@n3))
   > SetBoxState(self=@n1.ShortcutBox, Checked=@n3)
  ]}
 RefreshAll {nodes [
   RefreshAll!() -> (then, ImmediateUpdate:bool=~) {is=title K2Node_FunctionEntry}
   > SetDynMeshMaterial()
   > n1 = Set GizmoVizAtStart(GizmoVizAtStart:bool=Get GizmosVisible() -> (GizmosVisible:bool)) -> (Output_Get:bool)
   > n2 = Branch(Condition=@n1)
   > n3 = GizmoVizControl(Visible=false)
   UpdateSplinePoints!(execute=[@n3.then, @n2.else], self, Points:Array of Vectors=Get AllSplinePoints()) -> (then)
   n4 = Get PolylinePoints() -> (PolylinePoints:Array of Vectors)
   > Clear(TargetArray=@n4)
   > ConvertSplineToPolyline!(execute, self, TargetArray:Array of Vectors=@n4) -> (then) {is=title K2Node_CallFunction}
   > n5 = Branch(Condition=Get ImmediateUpdate() -> (ImmediateUpdate:bool))
   n6 = Switch on E_TravelTime_ReprojectOptions(execute=[@n11.then, @n5.then], Selection=Get Follow Surface(self=Get Settings()) -> (Follow Surface=NewEnumerator0))
   > n7 = SweepSpline!(execute, self, SweepPath:Array of Vectors=Get PolylinePoints()) -> (then)
   UpdateBillboards!(execute=[@n7.then, @n9.then], self, Spline Points:Array of Vectors=Get AllSplinePoints()) -> (then)
   > n8 = SetOverlayInfo(IsWarning=false, UseMessageLine=false, Total Dist=Get TotalDist(), Total Time=Get TotalTime())
   Branch(execute=[@n8.then, @n11.else], Condition=Get GizmoVizAtStart() -> (GizmoVizAtStart:bool))
   > GizmoVizControl(Visible=true)
   Clear(
    execute=[@n6.NewEnumerator1, @n6.NewEnumerator2]
    TargetArray=Get RreprojectedPolylinePoints() -> (RreprojectedPolylinePoints:Array of Vectors)
   )
   > ReprojectSpline()
   > n9 = SweepSpline(SweepPath=Get RreprojectedPolylinePoints())
   n10 = WaitForTime!(execute=@n5.else, self, PauseTime:double=Get WaitForTimePause() -> (WaitForTimePause:double)) -> (then, Active:bool)
   > n11 = Branch(Condition=@n10)
  ]}
 ReprojectSpline {nodes [
   ReprojectSpline() {is=title K2Node_FunctionEntry}
   > SetSplineCompViz()
   n1 = Get PolylinePoints()
   > Branch(Condition="integer > integer"(A=Length(TargetArray=@n1), B=1))
   > Clear(TargetArray=Get RreprojectedPolylinePoints())
   > Set FoundLandscapeWholeSpline(FoundLandscapeWholeSpline:bool=false) -> (Output_Get:bool)
   > SubdivideSpline!(
    execute
    self
    SourceSplineComp:Spline Component Object Reference=Get SplineComponent()
    DestSplineComp:Spline Component Object Reference=Get ReprojectSplineComponent()
    DistBetweenPoints:double=Get SpacingForSplineResample() -> (SpacingForSplineResample:double)
   ) -> (then)
   n2 = Get ReprojectSplineComponent()
   > n3 = For Loop(LastIndex=int - int(A=GetNumberOfSplinePoints(self=@n2) -> (ReturnValue=0)))
   SetSplinePoints!(
    execute=@n3.Completed
    self:Spline Component Object Reference=Get ReprojectSplineComponent()
    Points:Array of Vectors=Get RreprojectedPolylinePoints()
    CoordinateSpace:ESplineCoordinateSpace Enum=World
    bUpdateSpline:bool=true
   ) -> (then)
   > SetOverrideConstructionScript!(execute, self:Spline Component Object Reference=Get ReprojectSplineComponent(), InOverride:bool=true) -> (then)
   > UpdateSpline(self:Spline Component Object Reference=Get ReprojectSplineComponent())
   > n4 = Switch on E_TravelTime_ReprojectOptions(Selection=Get Follow Surface(self=Get Settings()) -> (Follow Surface=NewEnumerator0))
   n5 = Branch(execute=@n4.NewEnumerator1, Condition=Get FoundLandscapeWholeSpline() -> (FoundLandscapeWholeSpline:bool))
   SetOverlayInfo(
    execute=@n5.else
    TextToOverlay=Format Text(Format=Get NoLandscapeFoundWarning() -> (NoLandscapeFoundWarning:Text), 0=MakeLiteralText(Value="LOCTEXT(\"landscape\")"))
    IsWarning=true
    UseMessageLine=false
    Total Dist=0
    Total Time=0
   )
   > Set Last Clear Time(Last Clear Time=Get CurTime())
   Set FoundLandscapeThisPoint(execute=@n3.LoopBody, FoundLandscapeThisPoint:bool=false) -> (Output_Get:bool)
   > n6 = Set CurOriginalLocation(CurOriginalLocation:Vector=GetSplinePointAt(self=Get ReprojectSplineComponent(), PointIndex=@n3, CoordinateSpace=World) -> (
     ReturnValue_InputKey=~
     ReturnValue_Position=~
     ReturnValue_ArriveTangent=~
     ReturnValue_LeaveTangent=~
     ReturnValue_Rotation=~
     ReturnValue_Scale=~
    ).ReturnValue_Position) -> (Output_Get:Vector)
   > n7 = Switch on E_TravelTime_ReprojectOptions(Selection=Get Follow Surface(self=Get Settings()) -> (Follow Surface=NewEnumerator0))
   n8 = BreakVector(InVec=@n6)
   n9 = MakeVector(X=@n8, Y=@n8.Y, Z=Get ReprojectZOffset() -> (ReprojectZOffset:double))
   n10 = MakeVector(X=@n8, Y=@n8.Y, Z=float * float(A=Get ReprojectZOffset(), B=-1))
   n11 = Multi Line Trace For Objects!(
    execute=@n7.NewEnumerator1
    Start:Vector=@n9
    End:Vector=@n10
    ObjectTypes:Array of EObjectTypeQuery Enums=Make Array!(
     "[0]":EObjectTypeQuery Enum=ObjectTypeQuery1
     "[1]":EObjectTypeQuery Enum=ObjectTypeQuery2
     "[2]":EObjectTypeQuery Enum=ObjectTypeQuery5
     "[3]":EObjectTypeQuery Enum=ObjectTypeQuery6
    ) -> (Array:Array of EObjectTypeQuery Enums)
    bTraceComplex:bool=true
    ActorsToIgnore:Array of Actor Object References
    DrawDebugType:EDrawDebugTrace Enum
    bIgnoreSelf:bool=true
    TraceColor:Linear Color Structure=(R=1,G=0,B=0,A=1)
    TraceHitColor:Linear Color Structure=(R=0,G=1,B=0,A=1)
    DrawTime:float=5
   ) -> (then, OutHits:Array of Hit Result Structures, ReturnValue:bool)
   > n12 = For Each Loop with Break(Array:Array of Hit Result Structures=@n11, Break=@n15.then) -> (Array Element:Hit Result Structure)
   n13 = BreakHitResult(Hit=@n12)
   > Branch(Condition=IsValid(Object=@n13.HitActor))
   n14 = Get Class(Object=@n13.HitActor)
   > Branch(Condition=Contains(SearchIn=Get Class Display Name(Class=@n14), Substring=landscape))
   > Add(TargetArray:Array of Vectors=Get RreprojectedPolylinePoints(), NewItem:"Vector (by ref)"=@n13.Location)
   > n15 = Set FoundLandscapeThisPoint(FoundLandscapeThisPoint=true)
   n16 = Branch(execute=@n12.Completed, Condition=Get FoundLandscapeThisPoint() -> (FoundLandscapeThisPoint:bool))
   > Set FoundLandscapeWholeSpline(FoundLandscapeWholeSpline=true)
   n17 = Line Trace By Channel(execute=[@n16.else, @n7.NewEnumerator2], Start=@n9, End=@n10, bTraceComplex=true, ActorsToIgnore=Get All Measure Tool Spline Actors EUA())
   n18 = BreakHitResult(Hit=@n17)
   > n19 = Branch(Condition=@n18)
   > Add(TargetArray=Get RreprojectedPolylinePoints(), NewItem=@n18.Location)
   Add(execute=@n19.else, TargetArray=Get RreprojectedPolylinePoints(), NewItem=Get CurOriginalLocation() -> (CurOriginalLocation:Vector))
  ]}
 SaveSplineGeoToAsset {nodes [
   SaveSplineGeoToAsset() {is=title K2Node_FunctionEntry}
   > Sequence!(execute) -> (then_0, then_1)
   n1 = Get SplineActorEUA()
   > Is Valid(InputObject=@n1)
   > n2 = GetDynamicMesh(self=Get Dynamic Mesh Component())
   > Branch(Condition="integer > integer"(A=GetTriangleCount!(self:Dynamic Mesh Object Reference=@n2) -> (ReturnValue:int=0)))
   n3 = GetActorLabel(self=@n1, bCreateIfNone=false)
   > n4 = CreateNewStaticMeshAssetFromMesh!(
    execute
    FromDynamicMesh:Dynamic Mesh Object Reference=@n2
    AssetPathAndName:String=Append!(A=Get TempStaticMeshAssetBasePath() -> (TempStaticMeshAssetBasePath:String), B=@n3, C=., D:String=@n3) -> (ReturnValue)
    Options:Geometry Script Create New Static Mesh Asset Options Structure
   ) -> (Failure:Exec, Success:Exec, ReturnValue:Static Mesh Object Reference)
   > PrintString!(
    execute
    InString=Travel Time: Create New Static Mesh Failed
    bPrintToScreen=false
    bPrintToLog
    TextColor=(R=0.7315,G=0,B=0,A=1)
    Duration=5
    Key
   ) -> (then)
   SetStaticMesh(execute=@n4.Success, self=Get Static Mesh Component(), NewMesh=@n4)
   > SetMaterial(
    self=Get Static Mesh Component()
    ElementIndex=0
    Material=GetMaterial!(self:Primitive Component Object Reference=Get Dynamic Mesh Component(), ElementIndex:int=0) -> (ReturnValue:Material Interface Object Reference)
   )
   > Reset!(execute, self:Dynamic Mesh Object Reference=@n2) -> (then, ReturnValue:Dynamic Mesh Object Reference)
  ]}
 SetComponentEditorOnly {nodes [
   n1 = SetComponentEditorOnly!() -> (then, Object:Object Reference) {is=title K2Node_FunctionEntry}
   > SetEditorProperty(Object=@n1, PropertyName=bIsEditorOnly, PropertyValue=Get EditorOnly())
  ]}
 SetDynMeshMaterial {nodes [
   SetDynMeshMaterial() {is=title K2Node_FunctionEntry}
   > n1 = Branch(Condition=Get OpaqueLines!(self:BP STPS Travel Time Object Reference=Get Settings()) -> (OpaqueLines:bool))
   n2 = Get Dynamic Mesh Component()
   > n3 = SetMaterial(self=@n2, ElementIndex=0, Material=/Game/Blueprints/Blutilities/ToolsWrkshp/ScrTools/TravelTime/Materials/MI/MI_TravelTime_TextMatOpaque)
   Return Node!(execute=[@n3.then, @n4.then])
   n4 = SetMaterial(
    execute=@n1.else
    self=@n2
    ElementIndex=0
    Material=/Game/Blueprints/Blutilities/ToolsWrkshp/ScrTools/TravelTime/Materials/MI/MI_TravelTime_TextMat
   )
  ]}
 SetOverlayInfo {nodes [
   n1 = SetOverlayInfo!() -> (then, TextToOverlay:Text, IsWarning:bool=~, UseMessageLine:bool=~, Total Dist:double=~, Total Time:double=~) {is=title K2Node_FunctionEntry}
   > n2 = Branch(Condition=@n1.IsWarning)
   > DisplayUserWarningMessage(Message=Get TextToOverlay() -> (TextToOverlay:Text))
   n3 = Get Settings()
   Set DistUnits(
    execute=@n2.else
    DistUnits:String=Enum to String(Enumerator:E_TravelTime_DistUnits Enum=Get Distance Units(self=@n3) -> (Distance Units=NewEnumerator5))
   ) -> (Output_Get:String)
   > Set TimeUnits(TimeUnits:String=Enum to String(Enumerator:E_TravelTime_TimeUnits Enum=Get Time Units(self=@n3) -> (Time Units=NewEnumerator5))) -> (Output_Get:String)
   > n4 = GetConvertedUnits(
    Distance in cm=Get Total Dist() -> (Total Dist:double)
    Distance Units=Get DistUnits() -> (DistUnits:String)
    Time in sec=Get Total Time() -> (Total Time:double)
    Time Units=Get TimeUnits() -> (TimeUnits:String)
    Grid Units="To Float (Integer)"(InInt=Get Grid Snap Size(self=@n3) -> (Grid Snap Size=0))
   )
   > n5 = Sequence()
   > SetLine!(
    execute
    self:BP UW Readout Line Object Reference=Get Distance Line() -> (Distance Line:BP UW Readout Line Object Reference)
    Label:Text=Format Text(Format=Get Widget Distance Line Label(), 0:String=@n4.Distance Units Final)
    Value:Text="To Text (Float)"(Value=@n4, MinimumIntegralDigits=1, MaximumIntegralDigits=324, MinimumFractionalDigits=0, MaximumFractionalDigits=4)
   ) -> (then)
   SetLine(
    execute=@n5.then_1
    self=Get Time Line() -> (Time Line:BP UW Readout Line Object Reference)
    Label=Format Text(Format=Get Widget Time Line Label(), 0=@n4.Time Units Final)
    Value="To Text (Float)"(Value=@n4.Converted time, MinimumIntegralDigits=1, MaximumIntegralDigits=324, MinimumFractionalDigits=0, MaximumFractionalDigits=4)
   )
  ]}
 SetSplineCompViz {nodes [
   SetSplineCompViz() {is=title K2Node_FunctionEntry}
   > n1 = Switch on E_TravelTime_ReprojectOptions(Selection=Get Follow Surface(self=Get Settings()) -> (Follow Surface=NewEnumerator0))
   > SetVisibility(self=Get SplineComponent(), bNewVisibility=true, bPropagateToChildren=false)
   > SetVisibility(self=Get ReprojectSplineComponent(), bNewVisibility=false, bPropagateToChildren=false)
   SetVisibility(execute=[@n1.NewEnumerator1, @n1.NewEnumerator2], self=Get SplineComponent(), bNewVisibility=false, bPropagateToChildren=false)
   > SetVisibility(self=Get ReprojectSplineComponent(), bNewVisibility=true, bPropagateToChildren=false)
  ]}
 ShouldDoEdit {nodes [
   ShouldDoEdit() {is=title K2Node_FunctionEntry}
   > n1 = GetSelectionSet(self=Get LevelEditorSubsystem())
   > n2 = GetSelectedObjects!(
    execute
    self:Typed Element Selection Set Object Reference=@n1
    InRequiredClass:Object Class Reference=&BP_EUA_TravelTime_SplineActor_C
   ) -> (then, ReturnValue:Array of Object References)
   n3 = Length(TargetArray:Array of Object References=@n2)
   > n4 = Branch(Condition="Equal (Integer)"(A=@n3))
   > n5 = Set DoEdit(DoEdit:bool=false) -> (Output_Get:bool)
   n6 = Cast To BP_EUA_TravelTime_SplineActor!(execute=@n9.then, Object="Get (a copy)"(Array:Array of Object References=@n2) -> (Output:Object Reference)) -> (then, CastFailed, AsBP EUA Travel Time Spline Actor:BP EUA Travel Time Spline Actor Object Reference) {is=title K2Node_DynamicCast}
   Return Node!(
    execute=[@n8.then, @n10.then, @n5.then, @n6.then]
    ShouldDoEdit:bool=Get DoEdit()
    Selected Spline Actor:BP EUA Travel Time Spline Actor Object Reference=@n6
   )
   n7 = Branch(execute=@n4.else, Condition="integer > integer"(A=@n3, B=1))
   > Set DoEdit(DoEdit=false)
   > n8 = SetOverlayInfo(
    TextToOverlay=Format Text(Format=Get EditTooManyActorsSelWarn1() -> (EditTooManyActorsSelWarn1:Text), 0:int=@n3)
    IsWarning=true
    UseMessageLine=true
    Total Dist=0
    Total Time=0
   ) {is=title K2Node_CallFunction}
   n9 = Set DoEdit(execute=@n7.else, DoEdit=true)
   n10 = Set DoEdit(execute=@n6.CastFailed, DoEdit=false)
  ]}
 SubdivideSpline {nodes [
   n1 = SubdivideSpline!() -> (then, SourceSplineComp:Spline Component Object Reference, DestSplineComp:Spline Component Object Reference, DistBetweenPoints:double=100) {is=title K2Node_FunctionEntry}
   > Set SourceSplineLocal(SourceSplineLocal:Spline Component Object Reference=@n1) -> (Output_Get:Spline Component Object Reference)
   > Set DestSplineCompLocal(DestSplineCompLocal:Spline Component Object Reference=@n1.DestSplineComp) -> (Output_Get:Spline Component Object Reference)
   > Set DistBetweenPointsLocal(DistBetweenPointsLocal:double=@n1.DistBetweenPoints) -> (Output_Get:double)
   > n2 = Sequence!(execute) -> (then_0, then_1, then_2:Exec, then_3:Exec)
   > Set OrigSplineLength(OrigSplineLength:float=GetSplineLength(self=Get SourceSplineLocal() -> (SourceSplineLocal:Spline Component Object Reference))) -> (Output_Get:float)
   > Set CurDist(CurDist:double=0) -> (Output_Get:double)
   > Set TotalDistTraveled(TotalDistTraveled:double=0) -> (Output_Get:double)
   > ClearSplinePoints(self=Get DestSplineCompLocal() -> (DestSplineCompLocal:Spline Component Object Reference), bUpdateSpline=true)
   > Clear(TargetArray=Get TempSplinePointLocations() -> (TempSplinePointLocations:Array of Vectors))
   SetSplinePoints(execute=@n2.then_2, self=Get DestSplineCompLocal(), Points=Get TempSplinePointLocations(), CoordinateSpace=World, bUpdateSpline=true)
   > n3 = Branch(Condition=Get Use Straight Lines!(self:BP STPS Travel Time Object Reference=Get Settings()) -> (Use Straight Lines:bool))
   n4 = Get DestSplineCompLocal()
   > n5 = For Loop(LastIndex=int - int(A=GetNumberOfSplinePoints(self=@n4) -> (ReturnValue=0)))
   SetOverrideConstructionScript(execute=[@n3.else, @n5.Completed], self=Get DestSplineCompLocal(), InOverride=true)
   > UpdateSpline(self=Get DestSplineCompLocal())
   SetSplinePointType!(
    execute=@n5.LoopBody
    self:Spline Component Object Reference=Get DestSplineCompLocal()
    PointIndex:int=@n5
    Type:ESplinePointType Enum=Linear
    bUpdateSpline:bool=false
   ) -> (then)
   Set SubdividedSplinePointLocations(execute=@n2.then_3, SubdividedSplinePointLocations:Array of Vectors=Get TempSplinePointLocations()) -> (Output_Get:Array of Vectors)
   > Set SubdividedSplinePointSegmentNum(SubdividedSplinePointSegmentNum:Array of Integers=Get TempSplinePointLocSegmentNum() -> (TempSplinePointLocSegmentNum:Array of Integers)) -> (Output_Get:Array of Integers)
   > Return Node!(execute)
   n6 = Get OrigSplineLength() -> (OrigSplineLength:float)
   n7 = Get DistBetweenPointsLocal() -> (DistBetweenPointsLocal:double)
   n8 = For Loop with Break!(Execute:Exec=@n2.then_1, FirstIndex:int, LastIndex:int=Ceil(A=SafeDivide(A=@n6, B=@n7)), Break:Exec) -> (LoopBody:Exec, Index:int, Completed:Exec) {is=title K2Node_MacroInstance}
   n9 = "Clamp (Float)"(Value=float * float(A:int=@n8, B=Get DistBetweenPointsLocal()), Min=0, Max=Get OrigSplineLength())
   > Add(
    TargetArray=Get TempSplinePointLocations()
    NewItem=GetLocationAtDistanceAlongSpline!(self:Spline Component Object Reference=Get SourceSplineLocal(), Distance:float=@n9, CoordinateSpace:ESplineCoordinateSpace Enum=World) -> (ReturnValue:Vector)
   )
   n10 = GetInputKeyValueAtDistanceAlongSpline!(self:Spline Component Object Reference=Get SourceSplineLocal(), Distance:float=@n9) -> (ReturnValue:float)
   n11 = GetNumberOfSplineSegments!(self:Spline Component Object Reference=Get SourceSplineLocal()) -> (ReturnValue:int=0)
   > Add(
    TargetArray:Array of Integers=Get TempSplinePointLocSegmentNum()
    NewItem:"Integer (by ref)"="Clamp (Integer)"(Value=Floor(A=@n10), Min=0, Max=int - int(A=@n11)) -> (ReturnValue=0)
   )
  ]}
 SweepSpline {nodes [
   SweepSpline!() -> (then, SweepPath:Array of Vectors) {is=title K2Node_FunctionEntry}
   > n1 = GetDynamicMesh(self=Get Dynamic Mesh Component())
   > n2 = Set CurDynMesh(CurDynMesh:Dynamic Mesh Object Reference=@n1) -> (Output_Get:Dynamic Mesh Object Reference)
   > Reset(self=@n2)
   n3 = Get SweepPath() -> (SweepPath:Array of Vectors)
   > Branch(Condition="integer > integer"(A=Length(TargetArray:Array of Vectors=@n3), B=1))
   > n4 = Sequence!(execute) -> (then_0, then_1)
   > Clear(TargetArray:Array of Integers=Get PolylinePointsClosestSplinePoint() -> (PolylinePointsClosestSplinePoint:Array of Integers))
   > n5 = For Each Loop(Array:Array of Vectors=Get SweepPath()) -> (Array Element:Vector)
   n6 = FindInputKeyClosestToWorldLocation(self=Get SplineComponent(), WorldLocation=@n5)
   n7 = GetNumberOfSplineSegments(self=Get SplineComponent()) -> (ReturnValue=0)
   > Add(
    TargetArray=Get PolylinePointsClosestSplinePoint()
    NewItem="Clamp (Integer)"(Value=Floor(A=@n6), Min=0, Max=int - int(A=@n7)) -> (ReturnValue=0)
   )
   n8 = Get Segment Line Thickness!(self:BP STPS Travel Time Object Reference=Get Settings()) -> (Segment Line Thickness:double)
   n9 = AppendSimpleSweptPolygon!(
    execute=@n4.then_1
    TargetMesh:Dynamic Mesh Object Reference=Get CurDynMesh() -> (CurDynMesh:Dynamic Mesh Object Reference)
    PrimitiveOptions_PolygroupMode:EGeometryScriptPrimitivePolygroupMode Enum=PerFace
    PrimitiveOptions_bFlipOrientation:bool=False
    PrimitiveOptions_UVMode:EGeometryScriptPrimitiveUVMode Enum=Uniform
    PrimitiveOptions_MaterialID:int=0
    Transform:Transform=MakeTransform(Rotation="0, 0, 0")
    PolygonVertices:Array of Vector 2D Structures=Make Array(
     "[0]":Vector 2D Structure=(X=-0.5,Y=-0.5)
     "[1]":Vector 2D Structure=(X=0.5,Y=-0.5)
     "[2]":Vector 2D Structure=(X=0.5,Y=0.5)
     "[3]":Vector 2D Structure=(X=-0.5,Y=0.5)
    ) -> (Array:Array of Vector 2D Structures)
    SweepPath:Array of Vectors=Get SweepPath()
    bLoop:bool=false
    bCapped:bool=true
    StartScale:float=@n8
    EndScale:float=@n8
    RotationAngleDeg:float=0
    MiterLimit:float=1
   ) -> (then, ReturnValue:Dynamic Mesh Object Reference)
   > Clear(TargetArray:Array of Linear Color Structures=Get ColorArray() -> (ColorArray:Array of Linear Color Structures))
   > n10 = For Loop(LastIndex=int - int(A=GetVertexCount!(TargetMesh:Dynamic Mesh Object Reference=@n9) -> (ReturnValue:int=0)))
   n11 = ConvertArrayToColorList!(execute=@n10.Completed, ColorArray:Array of Linear Color Structures=Get ColorArray()) -> (then, ColorList:Color List Structure)
   > SetMeshPerVertexColors!(execute, TargetMesh:Dynamic Mesh Object Reference=Get CurDynMesh(), VertexColorList:Color List Structure=@n11) -> (then, ReturnValue:Dynamic Mesh Object Reference)
   n12 = FindInputKeyClosestToWorldLocation(
    self=Get SplineComponent()
    WorldLocation=GetVertexPosition!(TargetMesh:Dynamic Mesh Object Reference=Get CurDynMesh(), VertexID:int=@n10) -> (bIsValidVertex:bool, ReturnValue:Vector).ReturnValue
   )
   n13 = GetNumberOfSplineSegments(self=Get SplineComponent()) -> (ReturnValue=0)
   n14 = "% (Integer)"(A="Clamp (Integer)"(Value=Floor(A=@n12), Min=0, Max=int - int(A=@n13)) -> (ReturnValue=0), B=2) -> (ReturnValue=0)
   n15 = Get Settings()
   n16 = Get Segment 1 Color!(self:BP STPS Travel Time Object Reference=@n15) -> (Segment 1 Color:Color Structure)
   n17 = Get Segment 2 Color!(self:BP STPS Travel Time Object Reference=@n15) -> (Segment 2 Color:Color Structure)
   Add(
    execute=@n10.LoopBody
    TargetArray:Array of Linear Color Structures=Get ColorArray()
    NewItem:"Linear Color Structure (by ref)"=SelectColor(
     A="To LinearColor (Color)"!(InColor:Color Structure=@n17) -> (ReturnValue:Linear Color Structure)
     B="To LinearColor (Color)"(InColor=@n16)
     bPickA="To Boolean (Integer)"!(InInt:int=@n14) -> (ReturnValue:bool)
    )
   )
  ]}
 TimedClearUserWarning {nodes [
   n1 = TimedClearUserWarning!() -> (then, ClearAfterTime:double=3, CurTime:double=~, LastClearTime:double=~) {is=title K2Node_FunctionEntry}
   > Set LastClearTimeUpdated(LastClearTimeUpdated:double=@n1.LastClearTime) -> (Output_Get:double)
   > Branch(Condition="float > float"(A=@n1.CurTime, B=float + float(A=@n1, B=@n1.LastClearTime)))
   > Branch()
   > Set LastClearTimeUpdated(LastClearTimeUpdated=@n1.CurTime)
   > ClearUserMessages!(execute, self:Scriptable Interactive Tool Object Reference, bHelpMessage:bool=true, bWarningMessage:bool=true) -> (then)
   > Return Node!(execute, NewLastClearTime:double=Get LastClearTimeUpdated() -> (LastClearTimeUpdated:double))
  ]}
 UpdateBillboards {nodes [
   n1 = UpdateBillboards!() -> (then, Spline Points:Array of Vectors) {is=title K2Node_FunctionEntry}
   n2 = Length(TargetArray=@n1)
   > Branch(Condition="integer > integer"(A=@n2, B=1))
   > K2Node_Composite_1 = ClearSplineActorVariables!(execute) -> (then:Exec) {is=title K2Node_Composite}
   > Set NumSegments(NumSegments:int=int - int(A=@n2)) -> (Output_Get:int=0)
   > n3 = Set SplineLength(SplineLength:float=GetSplineLength(self=Get ReprojectSplineComponent())) -> (Output_Get:float)
   n4 = Get Settings()
   > n5 = GetTimeForDist!(
    execute
    self
    Dist:double=@n3
    CharSpeed:double=float * float(
     A:double=Get Movement Speed(self=@n4)
     B=Get Movement Speed Multiplier!(self:BP STPS Travel Time Object Reference=@n4) -> (Movement Speed Multiplier:double)
    )
    Units:E_TravelTime_SpeedUnits Enum=NewEnumerator2
   ) -> (then, Time:double, UnitsToDisplay:String) {is=title K2Node_CallFunction}
   > n6 = Set TotalRunTime!(execute, TotalRunTime:double=@n5, self:BP EUA Travel Time Spline Actor Object Reference=Get SplineActorEUA()) -> (then, Output_Get:double)
   > Set CurTotalRunTime(CurTotalRunTime:double=@n6) -> (Output_Get:double)
   > n7 = For Loop(LastIndex=int - int(A=Get NumSegments() -> (NumSegments:int=0)))
   n8 = Sequence!(execute=@n7.Completed) -> (then_0, then_1, then_2:Exec)
   > n9 = Set TotalDist(TotalDist=Get SplineLength() -> (SplineLength:float))
   > n10 = Set TotalTime(TotalTime=Get CurTotalRunTime() -> (CurTotalRunTime:double))
   > SetOverlayInfo(IsWarning=false, UseMessageLine=false, Total Dist=@n9, Total Time=@n10)
   Set AllTextRenderComponents!(
    execute=@n8.then_1
    AllTextRenderComponents:Array of Text Render Component Object References=Get AllTextRenderComponents() -> (AllTextRenderComponents:Array of Text Render Component Object References)
    self:BP EUA Travel Time Spline Actor Object Reference=Get SplineActorEUA()
   ) -> (then, Output_Get:Array of Text Render Component Object References)
   > Set AllBGRenderComponents!(
    execute
    AllBGRenderComponents:Array of Text Render Component Object References=Get AllBGRenderComponents() -> (AllBGRenderComponents:Array of Text Render Component Object References)
    self:BP EUA Travel Time Spline Actor Object Reference=Get SplineActorEUA()
   ) -> (then, Output_Get:Array of Text Render Component Object References)
   n11 = Set CurrentSegmentNumber(execute=@n7.LoopBody, CurrentSegmentNumber=@n7) -> (Output_Get=0)
   > Set DistAlongSplineAtCurPoint(DistAlongSplineAtCurPoint:float=GetDistanceAlongSplineAtSplinePoint(self=Get SplineComponent(), PointIndex=@n11)) -> (Output_Get:float)
   > n12 = Sequence!(execute) -> (then_0, then_1, then_2, then_3:Exec, then_4:Exec)
   > n13 = GetSubDSplineSegmentStartEndPointNumbers!(execute, self, SegNum:int=Get CurrentSegmentNumber() -> (CurrentSegmentNumber=0)) -> (then, ThisSegStartPointIndex1:int=0, ThisSegEndPointIndex1:int=0, CurrentSegmentStartPointLoc1:Vector, CurrentSegmentEndPointLoc1:Vector) {is=title K2Node_CallFunction}
   > Set ThisSegStartPointIndexLocal(ThisSegStartPointIndexLocal:int=@n13) -> (Output_Get:int=0)
   > Set ThisSegEndPointIndexLocal(ThisSegEndPointIndexLocal:int=@n13.ThisSegEndPointIndex1) -> (Output_Get:int=0)
   > Set CurrentSegmentStartPoint(CurrentSegmentStartPoint:Vector=@n13.CurrentSegmentStartPointLoc1) -> (Output_Get:Vector)
   > n14 = Set CurrentSegmentEndPoint(CurrentSegmentEndPoint:Vector=@n13.CurrentSegmentEndPointLoc1) -> (Output_Get:Vector)
   > n15 = CalcLineCentre!(execute, self, Point0:Vector=Get CurrentSegmentStartPoint() -> (CurrentSegmentStartPoint:Vector), Point1:Vector=@n14) -> (then, CentrePoint:Vector) {is=title K2Node_CallFunction}
   > Set Current Centre Point(Current Centre Point:Vector=@n15) -> (Output_Get:Vector)
   K2Node_Composite_0 = UpdateSplineActorArrays!(execute=@n12.then_4) -> (then:Exec) {is=title K2Node_Composite}
   n16 = GetLineLength!(
    execute=@n12.then_1
    self
    Start Point:Vector=Get CurrentSegmentStartPoint()
    End Point:Vector=Get CurrentSegmentEndPoint() -> (CurrentSegmentEndPoint:Vector)
    Segment Number:int=Get CurrentSegmentNumber() -> (CurrentSegmentNumber=0)
    SubDsegStartPointIndex:int=Get ThisSegStartPointIndexLocal() -> (ThisSegStartPointIndexLocal:int=0)
    SubDsegEndPointIndex:int=Get ThisSegEndPointIndexLocal() -> (ThisSegEndPointIndexLocal:int=0)
   ) -> (then, SubDSplineSegmentLength:double, SplineSegmentLen:double, Straight Dist:double) {is=title K2Node_CallFunction}
   > Set CurLineLen(CurLineLen:double=@n16) -> (Output_Get:double)
   n17 = Get Settings()
   > n18 = GetTimeForDist(Dist=@n16, CharSpeed=float * float(A=Get Movement Speed(self=@n17), B=Get Movement Speed Multiplier(self=@n17)), Units=NewEnumerator2)
   > Set CurLineRunTime(CurLineRunTime:double=@n18) -> (Output_Get:double)
   > n19 = MakeBillBoardTextToRender!(
    execute
    self
    CurLineLen:double=@n16.SplineSegmentLen
    CurLineRunTime:double=Get CurLineRunTime() -> (CurLineRunTime:double)
    SegmentNumber:int=Get CurrentSegmentNumber() -> (CurrentSegmentNumber=0)
    CurLineLenSubD:double=@n16
    CurLinLenStraight:double=@n16.Straight Dist
   ) -> (then, Result:Text) {is=title K2Node_CallFunction}
   > Set CurTextToRender(CurTextToRender:Text=@n19) -> (Output_Get:Text)
   n20 = Get SplineActorEUA()
   > Set CurTotalLen(CurTotalLen:double=float + float(A=Get TotalLen(self=@n20), B=Get CurLineLen() -> (CurLineLen:double))) -> (Output_Get:double)
   n21 = Branch(
    execute=@n12.then_3
    Condition=Is Valid Index!(
     TargetArray:Array of Text Render Component Object References=Get Starting BG Render Components() -> (Starting BG Render Components:Array of Text Render Component Object References)
     IndexToTest:int=Get CurrentSegmentNumber() -> (CurrentSegmentNumber=0)
    ) -> (ReturnValue:bool)
   )
   n22 = "To Transform (Vector)"(InLocation=vector + vector(A=Get Current Centre Point() -> (Current Centre Point:Vector), B=Get Spline Hit Z Tweak()))
   n23 = "Get (a copy)"(
    Array:Array of Text Render Component Object References=Get Starting BG Render Components()
    Dimension 1=Get CurrentSegmentNumber() -> (CurrentSegmentNumber=0)
   ) -> (Output:Text Render Component Object Reference)
   > Set World Transform(self=@n23, NewTransform=@n22, bSweep=false, bTeleport=false)
   > n24 = Set CurText BG Comp(CurText BG Comp:Text Render Component Object Reference=@n23) -> (Output_Get:Text Render Component Object Reference)
   Set CurTextBGDynMaterial(execute=[@n24.then, @n32.then], CurTextBGDynMaterial:Material Instance Dynamic Object Reference) -> (Output_Get:Material Instance Dynamic Object Reference)
   > n25 = CreateDynamicMaterialInstance!(
    execute
    Parent=/Game/Blueprints/Blutilities/ToolsWrkshp/ScrTools/TravelTime/Materials/M_TravelTime_CameraFacingCard
    OptionalName=MI_CameraFacingTextCard
    CreationFlags=Transient
   ) -> (then, ReturnValue)
   > n26 = Set CurTextBGDynMaterial(CurTextBGDynMaterial=@n25)
   > SetTextMaterial!(
    execute
    self:Text Render Component Object Reference=Get CurText BG Comp() -> (CurText BG Comp:Text Render Component Object Reference)
    Material:Material Interface Object Reference=@n26
   ) -> (then)
   > Set Text!(
    execute
    self:Text Render Component Object Reference=Get CurText BG Comp()
    Value:"Text (by ref)"=Get CurTextToRender() -> (CurTextToRender:Text)
   ) -> (then) {is=title K2Node_CallFunction}
   > SetWorldSize!(execute, self:Text Render Component Object Reference=Get CurText BG Comp(), Value:float=75) -> (then)
   > SetTranslucentSortPriority!(execute, self:Primitive Component Object Reference=Get CurText BG Comp(), NewTranslucentSortPriority:int=-2) -> (then)
   n27 = Get Settings()
   n28 = Get Text Background Opacity!(self:BP STPS Travel Time Object Reference=@n27) -> (Text Background Opacity:double)
   n29 = "Clamp (Integer64)"!(Value:Integer64=Round to Integer64!(A:double=float * float(A=@n28, B=255)) -> (ReturnValue:Integer64=0), Min:Integer64=0, Max:Integer64=255) -> (ReturnValue:Integer64=0)
   > SetTextRenderColor!(
    execute
    self:Text Render Component Object Reference=Get CurText BG Comp()
    Value:Color Structure=Make Color!(B:Byte=0, G:Byte=0, R:Byte=0, A:Byte="To Byte (Integer64)"!(InInt:Integer64=@n29) -> (ReturnValue:Byte=0)) -> (Color:Color Structure) {is=title K2Node_MakeStruct}
   ) -> (then)
   > SetHorizontalAlignment!(execute, self:Text Render Component Object Reference=Get CurText BG Comp(), Value:EHorizTextAligment Enum=EHTA_Center) -> (then)
   > Add(
    TargetArray:Array of Text Render Component Object References=Get AllBGRenderComponents()
    NewItem:"Text Render Component Object Reference (by ref)"=Get CurText BG Comp()
   )
   n30 = Get Settings()
   > SetScalarParameterValue(
    self=Get CurTextBGDynMaterial() -> (CurTextBGDynMaterial:Material Instance Dynamic Object Reference)
    ParameterName=camdistscale
    Value=float * float(A=Get Text Size!(self:BP STPS Travel Time Object Reference=@n30) -> (Text Size:double), B=Get BillboardScaleFactor())
   )
   > Add(
    TargetArray:Array of Material Instance Dynamic Object References=Get CurTextBGDynMaterials()
    NewItem:"Material Instance Dynamic Object Reference (by ref)"=Get CurTextBGDynMaterial()
   )
   n31 = MakeSubobjectTextRender!(execute=@n21.else, self, TheTransform:Transform=@n22) -> (then, Text Render Component:Text Render Component Object Reference) {is=title K2Node_CallFunction}
   > Set ComponentTags!(execute, ComponentTags:Array of Names=Make Array!("[0]":Name=BG) -> (Array:Array of Names), self:Actor Component Object Reference=@n31) -> (then, Output_Get:Array of Names)
   > n32 = Set CurText BG Comp(CurText BG Comp=@n31)
   n33 = Branch(
    execute=@n12.then_2
    Condition=Is Valid Index(
     TargetArray=Get StartingTextRenderComponents() -> (StartingTextRenderComponents:Array of Text Render Component Object References)
     IndexToTest=Get CurrentSegmentNumber() -> (CurrentSegmentNumber=0)
    )
   )
   n34 = "To Transform (Vector)"(InLocation=vector + vector(A=Get Current Centre Point(), B=Get Spline Hit Z Tweak()))
   n35 = "Get (a copy)"(Array=Get StartingTextRenderComponents(), Dimension 1=Get CurrentSegmentNumber() -> (CurrentSegmentNumber=0))
   > Set World Transform(self=@n35, NewTransform=@n34, bSweep=false, bTeleport=false)
   > n36 = Set CurTextRenderComp(CurTextRenderComp:Text Render Component Object Reference=@n35) -> (Output_Get:Text Render Component Object Reference)
   Set CurTextRenderDynMaterial(execute=[@n41.then, @n36.then], CurTextRenderDynMaterial:Material Instance Dynamic Object Reference) -> (Output_Get:Material Instance Dynamic Object Reference)
   > n37 = CreateDynamicMaterialInstance(
    Parent=/Game/Blueprints/Blutilities/ToolsWrkshp/ScrTools/TravelTime/Materials/M_TravelTime_CameraFacing_Transl
    OptionalName=MI_CameraFacingText_Translucent
    CreationFlags=Transient
   )
   > n38 = Set CurTextRenderDynMaterial(CurTextRenderDynMaterial=@n37)
   > SetMaterial(self=Get CurTextRenderComp() -> (CurTextRenderComp:Text Render Component Object Reference), ElementIndex=0, Material=@n38)
   > SetTextRenderColor(self=Get CurTextRenderComp(), Value=Get TextColor!(self:BP STPS Travel Time Object Reference=Get Settings()) -> (TextColor:Color Structure))
   > Set Text(self=Get CurTextRenderComp(), Value=Get CurTextToRender())
   > SetWorldSize(self=Get CurTextRenderComp(), Value=75)
   > SetTranslucentSortPriority(self=Get CurTextRenderComp(), NewTranslucentSortPriority=2)
   > SetHorizontalAlignment(self=Get CurTextRenderComp(), Value=EHTA_Center)
   > Add(
    TargetArray:Array of Text Render Component Object References=Get AllTextRenderComponents()
    NewItem:"Text Render Component Object Reference (by ref)"=Get CurTextRenderComp()
   )
   n39 = Get Settings()
   > SetScalarParameterValue(
    self=Get CurTextRenderDynMaterial() -> (CurTextRenderDynMaterial:Material Instance Dynamic Object Reference)
    ParameterName=camdistscale
    Value=float * float(A=Get Text Size(self=@n39), B=Get BillboardScaleFactor())
   )
   > Add(
    TargetArray:Array of Material Instance Dynamic Object References=Get CurTextRenderDynMaterials()
    NewItem:"Material Instance Dynamic Object Reference (by ref)"=Get CurTextRenderDynMaterial()
   )
   n40 = MakeSubobjectTextRender(execute=@n33.else, TheTransform=@n34)
   > Set ComponentTags(ComponentTags=Make Array("[0]"=Text), self=@n40)
   > n41 = Set CurTextRenderComp(CurTextRenderComp=@n40)
   n42 = Get AllTextRenderComponents()
   n43 = For Each Loop(Array:Array of Text Render Component Object References=@n42) -> (Array Element:Text Render Component Object Reference)
   n44 = "Equal (Integer)"(A=@n43.Array Index, B=int - int(A=Length(TargetArray:Array of Text Render Component Object References=@n42)))
   > SetVisibility(self=@n43, bNewVisibility=@n44, bPropagateToChildren=false)
   > SetVisibility(self="Get (a copy)"(Array=Get AllBGRenderComponents(), Dimension 1=@n43.Array Index), bNewVisibility=@n44, bPropagateToChildren=false)
   Break Color!(Color:"Color Structure (by ref)"=Get Text Background Color!(self:BP STPS Travel Time Object Reference=@n27) -> (Text Background Color:Color Structure)) -> (B:Byte, G:Byte, R:Byte, A:Byte) {is=title K2Node_BreakStruct}
   Get CurLineLen()
  ]}
 UpdateSplineActorArrays {
  path=UpdateBillboards.K2Node_Composite_0.UpdateSplineActorArrays
  nodes [
   Inputs!() -> (execute)
   > n1 = Sequence!(execute) -> (then_0, then_1, then_2, then_3)
   Add(
    execute=@n1.then_1
    TargetArray:Array of Strings=Get TextToRender!(self:BP EUA Travel Time Spline Actor Object Reference=Get SplineActorEUA()) -> (TextToRender:Array of Strings)
    NewItem:"String (by ref)"="To String (Text)"(InText=Get CurTextToRender())
   )
   > Add(
    TargetArray:"Array of Float (double-precision)s"=Get ThisLineLen!(self:BP EUA Travel Time Spline Actor Object Reference=Get SplineActorEUA()) -> (ThisLineLen:"Array of Float (double-precision)s")
    NewItem:"Float (double-precision) (by ref)"=Get CurLineLen()
   )
   > Add(
    TargetArray=Get ThisLineRunTime!(self:BP EUA Travel Time Spline Actor Object Reference=Get SplineActorEUA()) -> (ThisLineRunTime:"Array of Float (double-precision)s")
    NewItem=Get CurLineRunTime()
   )
   > Add(
    TargetArray:Array of Vectors=Get PlacedLocationBeforeOffset!(self:BP EUA Travel Time Spline Actor Object Reference=Get SplineActorEUA()) -> (PlacedLocationBeforeOffset:Array of Vectors)
    NewItem:"Vector (by ref)"=Get Current Centre Point()
   )
   Set TotalLen!(
    execute=@n1.then_2
    TotalLen:double=Get CurTotalLen() -> (CurTotalLen:double)
    self:BP EUA Travel Time Spline Actor Object Reference=Get SplineActorEUA()
   ) -> (then, Output_Get:double)
   Add(
    execute=@n1.then_0
    TargetArray:Array of Integers=Get LineNumber!(self:BP EUA Travel Time Spline Actor Object Reference=Get SplineActorEUA()) -> (LineNumber:Array of Integers)
    NewItem:"Integer (by ref)"=Get CurrentSegmentNumber() -> (CurrentSegmentNumber=0)
   )
   > Add(
    TargetArray:Array of S Travel Time Line Points Structures=Get LinePoints!(self:BP EUA Travel Time Spline Actor Object Reference=Get SplineActorEUA()) -> (LinePoints:Array of S Travel Time Line Points Structures)
    NewItem:"S Travel Time Line Points Structure (by ref)"=Make S Travel Time Line Points!(
     Start_2_#1:Vector=Get CurrentSegmentStartPoint()
     End_4_#2:Vector=Get CurrentSegmentEndPoint()
    ) -> (S_TravelTime_LinePoints:S Travel Time Line Points Structure) {is=title K2Node_MakeStruct}
   )
   Outputs!(then:Exec=@n1.then_3)
  ]
 }
 ClearSplineActorVariables {
  path=UpdateBillboards.K2Node_Composite_1.ClearSplineActorVariables
  nodes [
   Inputs()
   > n1 = Sequence!(execute) -> (then_0, then_1, then_2, then_3, then_4:Exec)
   Set TotalLen(execute=@n1.then_1, TotalLen=0, self=Get SplineActorEUA())
   > Set TotalRunTime(TotalRunTime=0, self=Get SplineActorEUA())
   Outputs(then=@n1.then_4)
   Clear(execute=@n1.then_0, TargetArray:Array of Integers=Get LineNumber(self=Get SplineActorEUA()))
   > Clear(TargetArray:Array of S Travel Time Line Points Structures=Get LinePoints(self=Get SplineActorEUA()))
   > Clear(TargetArray:Array of Strings=Get TextToRender(self=Get SplineActorEUA()))
   > Clear(TargetArray:"Array of Float (double-precision)s"=Get ThisLineLen(self=Get SplineActorEUA()))
   > Clear(TargetArray=Get ThisLineRunTime(self=Get SplineActorEUA()))
   > Clear(TargetArray:Array of Vectors=Get PlacedLocationBeforeOffset(self=Get SplineActorEUA()))
   > Clear(TargetArray:Array of Text Render Component Object References=Get AllBGRenderComponents())
   > Clear(TargetArray=Get AllTextRenderComponents())
   > Clear(TargetArray:Array of Material Instance Dynamic Object References=Get CurTextRenderDynMaterials())
   > Clear(TargetArray=Get CurTextBGDynMaterials())
   n2 = Set StartingTextRenderComponents(
    execute=@n1.then_2
    StartingTextRenderComponents:Array of Text Render Component Object References=GetComponentsByTag!(
     self:Actor Object Reference=Get SplineActorEUA()
     ComponentClass:Actor Component Class Reference=&TextRenderComponent
     Tag:Name=Text
    ) -> (ReturnValue:Array of Text Render Component Object References)
   ) -> (Output_Get:Array of Text Render Component Object References)
   > n3 = For Each Loop(Array=@n2)
   > SetWorldSize(self=@n3, Value=0)
   > Set Text(self=@n3, Value=Format Text!(Format) -> (Result))
   n4 = Set Starting BG Render Components(
    execute=@n1.then_3
    Starting BG Render Components:Array of Text Render Component Object References=GetComponentsByTag(self=Get SplineActorEUA(), ComponentClass=&TextRenderComponent, Tag=BG)
   ) -> (Output_Get:Array of Text Render Component Object References)
   > n5 = For Each Loop(Array=@n4)
   > SetWorldSize(self=@n5, Value=0)
   > Set Text(self=@n5, Value=Format Text())
  ]
 }
 UpdateSplinePoints {nodes [
   n1 = UpdateSplinePoints!() -> (then, Points:Array of Vectors) {is=title K2Node_FunctionEntry}
   > Branch(Condition="integer > integer"(A=Length(TargetArray:Array of Vectors=@n1)))
   > ClearSplinePoints(self=Get SplineComponent(), bUpdateSpline=false)
   > SetSplinePoints(self=Get SplineComponent(), Points=Get Points() -> (Points:Array of Vectors), CoordinateSpace=World, bUpdateSpline=true)
   > Set SplinePointPositions!(execute, SplinePointPositions:Array of Vectors=Get Points(), self:BP EUA Travel Time Spline Actor Object Reference=Get SplineActorEUA()) -> (then, Output_Get:Array of Vectors)
   > Set Num Spline Points(Num Spline Points:int=GetNumberOfSplinePoints(self=Get SplineComponent()) -> (ReturnValue=0)) -> (Output_Get:int=0)
   > Set ActorsThatShouldBeSelected(ActorsThatShouldBeSelected:Array of Actor Object References=Get ActorsSelectedAtStart()) -> (Output_Get:Array of Actor Object References)
   > n2 = Branch(Condition="Equal (Integer)"(A=Get Num Spline Points() -> (Num Spline Points:int=0), B=1))
   > n3 = Branch(Condition=Get IsEUAselected() -> (IsEUAselected:bool))
   n4 = Branch(execute=[@n7.then, @n8.else, @n3.then], Condition=Get Use Straight Lines(self=Get Settings()))
   > n5 = For Loop(LastIndex=int - int(A=Get Num Spline Points() -> (Num Spline Points=0)))
   SetOverrideConstructionScript(execute=[@n4.else, @n5.Completed], self=Get SplineComponent(), InOverride=true)
   > UpdateSpline(self=Get SplineComponent())
   > SubdivideSpline(SourceSplineComp=Get SplineComponent(), DestSplineComp=Get ReprojectSplineComponent(), DistBetweenPoints=Get SpacingForSplineResample()) {is=title K2Node_CallFunction}
   SetSplinePointType(execute=@n5.LoopBody, self=Get SplineComponent(), PointIndex=@n5, Type=Linear, bUpdateSpline=false)
   Add(
    execute=@n3.else
    TargetArray:Array of Actor Object References=Get ActorsThatShouldBeSelected() -> (ActorsThatShouldBeSelected:Array of Actor Object References)
    NewItem:"Actor Object Reference (by ref)"=Get SplineActorEUA()
   )
   > n6 = Set IsEUAselected(IsEUAselected=true)
   n7 = SetSelectedLevelActors(execute=[@n6.then, @n9.then], self=Get EditorActorSubsystem(), ActorsToSelect=Get ActorsThatShouldBeSelected())
   n8 = Branch(execute=@n2.else, Condition=Get IsEUAselected())
   > n9 = Set IsEUAselected(IsEUAselected=false)
  ]}
 WaitForTime {nodes [
   n1 = WaitForTime!() -> (then, PauseTime:double=0.5) {is=title K2Node_FunctionEntry}
   n2 = Get LastWaitTime() -> (LastWaitTime:double)
   > n3 = Branch(Condition="float > float"(A=Get CurTime(), B=float + float(A=@n1, B=@n2)))
   > Set LastWaitTime(LastWaitTime:double=Get CurTime()) -> (Output_Get:double)
   > Return Node!(execute, Active:bool=true)
   Return Node(execute=@n3.else, Active=false)
  ]}
 WarnGameViewOn {nodes [
   WarnGameViewOn() {is=title K2Node_FunctionEntry}
   > Branch(Condition=Get Is Game View On Tool Start())
   > DisplayUserWarningMessage(Message=Get GameViewOnWarnMessage() -> (GameViewOnWarnMessage:Text))
  ]}
]
variables [
 ActionButtons {type=BP_UW_ActionsButton_TravelTime_C, default=null}
 Actions Box {type=BP_UW_ExpandableBox_C}
 ActorsSelectedAtStart {type="array<Actor>"}
 All Measure Tool Spline Actors EUA {type="array<BP_EUA_TravelTime_SplineActor_C>", default []}
 AllBGRenderComponents {type="array<TextRenderComponent>", default []}
 AllGizmoNames {type="array<string>", default []}
 AllNonLandscapeActors {type="array<Actor>"}
 AllSplinePoints {type="array<Vector>"}
 AllTextRenderComponents {type="array<TextRenderComponent>", default []}
 BillboardScaleFactor {type=number, default=0.75}
 Clear Warning After Time {type=number, default=5}
 ClearSegmentsNoneFoundWarnMessage {type=string, default=No segments found to remove.}
 ClearSplinesNoneFoundWarnMessage {type=string, default=No splines found to remove.}
 ClearSplinesWarnIniName {type=string, default="ToolsWorkshop_{0}_ClearAllWarning"}
 ClearSplinesWarnMessage_p {type=string, default="This will remove {0} Travel Time splines from the level except the current spline object. Delete anyway?"}
 ClearSplinesWarnMessage_s {type=string, default=This will remove 1 Travel Time spline from the level except the current spline object. Delete anyway?}
 ClearSplinesWarnTitle {type=string, default=Clear Other Splines}
 CurTextBGDynMaterial {type=MaterialInstanceDynamic}
 CurTextBGDynMaterials {type="array<MaterialInstanceDynamic>", default [null]}
 CurTextRenderDynMaterial {type=MaterialInstanceDynamic}
 CurTextRenderDynMaterials {type="array<MaterialInstanceDynamic>", default [null]}
 CurTime {type=number}
 DefSceneRootHandle {type=SubobjectDataHandle, default {}}
 DeleteHotkey {type=Key, default=V}
 DeleteSegmentTransactionDescription {type=string, default=Delete Spline Point}
 Dist Along Spline Points {type="array<Vector>"}
 Distance Line {type=BP_UW_ReadoutLine_C, default=null}
 Distances Along Spline {type="array<number>", default []}
 DistConversionFactors {type="map<number>", default {cm=1, grids=-1, km=100000, m=100}}
 DoEdit {type=boolean}
 Dynamic Mesh Component {type=DynamicMeshComponent, default=null}
 EditModeHotkey {type=Key, default=X}
 EditorOnly {type=boolean, default=true}
 EditorOnlyTools {type=BP_TravelTime_EditorOnlyUtils_C, default=null}
 EditTooManyActorsSelWarn1 {type=string, default="Select only one spline to edit it. There are currently {0} selected."}
 Fort Speed Presets Map {
  type="map<number>"
  default {
   ADS Crouched Backwards=150
   ADS Crouched Forward=200
   ADS Crouched Side To Side=180
   ADS Walk Backwards=150
   ADS Walk Forward=200
   ADS Walk Side To Side=180
   Acceleration=700
   Braking Forward=850
   Braking Strafe=750
   Crouched Backwards=180
   Crouched Run=250
   Crouched Side To Side=200
   DBNO=168
   Jog Backwards=300
   Jog Shooting/Reloading=410
   Jog Side To Side=350
   Jump Z Velocity=550
   Run=550
   Tactical Sprint=750
   Walk=200
   Walk Backwards=150
   Walk Side To Side=180
  }
 }
 GameViewOnWarnMessage {type=string, default=Game View in the viewport has been turned off to view Travel Time splines. Press G in the viewport to turn it back on.}
 GizmoBaseIdent {type=string, default=PointGizmo_}
 GizmosVisible {type=boolean}
 GizmoTransformStateChangeSnapshot {type=Transform, default {location=<0 0 0>, rotation=<rot 0 0 0>, scale=<1 1 1>}}
 HaveAlreadyInitHeader {type=boolean}
 HaveAlreadyWarnedGameView {type=boolean}
 Header Widget {type=BP_UW_SharedHeaderWidget_C}
 Is Game View On Tool Start {type=boolean}
 IsDeleteKeyDown {type=boolean}
 isEditHotkeyDown {type=boolean}
 IsEUAselected {type=boolean}
 isPrependKeyDown {type=boolean}
 isSplitHotkeyDown {type=boolean}
 Last Clear Time {type=number}
 LastWaitTime {type=number}
 MouseClickButton {type=EScriptableToolMouseButton, default=LeftButton}
 NoLandscapeFoundWarning {type=string, default="No object with class name \"{0}\" found under spline to project onto."}
 OverlayTextblockFormat {type=string, default="Total Dist: {0}\r\nTotal Time: {1}"}
 Parent Handle {type=SubobjectDataHandle, default {}}
 PolylinePoints {type="array<Vector>"}
 PolylinePointsClosestSplinePoint {type="array<integer>"}
 PolyLineSpacing {type=number, default=100}
 PrependHotkey {type=Key, default=B}
 Previous Movement Speed {type=number}
 PreviousMovementPreset {type=E_TravelTime_SpeedPresets, default=Walk}
 PropSetSaveRestoreKey {type=string, default=MeasureToolLastSettings}
 Readout Box {type=BP_UW_ExpandableBox_C}
 ReprojectSplineComponent {type=SplineComponent, default=null}
 ReprojectZOffset {type=number, default=50000}
 RreprojectedPolylinePoints {type="array<Vector>"}
 Settings {type=BP_STPS_TravelTime_C, default=null}
 SettingsIdentifier {type=string, default=Settings}
 Shared State Settings {type=BP_STPS_SharedWidgetState_C, default=null}
 SharedPropSetWidgetStateIdent {type=string, default=SharedWidgetStatePropSet}
 Shortcut Box {type=BP_UW_ExpandableBox_C}
 Shortcut Map {
  type="map<string>"
  default {
   Add Segment To Beginning=Hold B + Left Click
   Delete Segment=Hold V + Left Click on Segment
   Divide Segment=Hold C + Left Click on Segment
   Move Segment=Hold X
  }
 }
 Shortcut Map Keys {type="array<string>", default [Move Segment, Divide Segment, Delete Segment, Add Segment To Beginning]}
 ShortcutLines {type="array<BP_UW_ReadoutLine_C>", default []}
 SingleClickHit {
  type=HitResult
  default {
   bBlockingHit=false
   bStartPenetrating=false
   boneName=None
   component=null
   distance=0
   elementIndex=0
   faceIndex=0
   item=0
   myBoneName=None
   myItem=-1
   penetrationDepth=0
   physMaterial=null
   time=1
   hitObjectHandle {referenceObject=null}
   impactNormal=<0 0 0>
   impactPoint=<0 0 0>
   location=<0 0 0>
   normal=<0 0 0>
   traceEnd=<0 0 0>
   traceStart=<0 0 0>
  }
 }
 SpacingForSplineResample {type=number, default=200}
 SpeedConversionFactors {type="map<number>", default {Grid/s=1, KPH=27.778, MPH=44.704, cm/s=1, m/s=100}}
 Spline Hit Z Tweak {type=Vector, default=<0 0 10>}
 SplineActorEUA {type=BP_EUA_TravelTime_SplineActor_C, default=null}
 SplineActorLabel {type=string, default=Travel Time Spline}
 SplineActorLabelRoot {type=string, default=Travel Time Spline}
 SplineComponent {type=SplineComponent, default=null}
 SplitHotkey {type=Key, default=C}
 Static Mesh Component {type=StaticMeshComponent, default=null}
 SubdividedSplinePointLocations {type="array<Vector>"}
 SubdividedSplinePointSegmentNum {type="array<integer>"}
 TempStaticMeshAssetBasePath {type=string, default=/Temp/MeasureTool/}
 Time Line {type=BP_UW_ReadoutLine_C, default=null}
 ToolStartTime {type=string, default=""}
 TotalDist {type=number}
 TotalTime {type=number}
 TransactionContext {type=string, default=""}
 Transient {type=boolean, default=true}
 WaitForTimePause {type=number, default=0.01}
 Widget Distance Line Label {type=string, default="Total Distance ({0})"}
 Widget Distance Line Value {type=string, default=0}
 Widget Time Line Label {type=string, default="Total Time ({0})"}
 Widget Time Line Value {type=string, default=0}
]
