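# M_TravelTime_TextMat is an unlit translucent material that, by its name and folder, serves text in the TravelTime editor scripting tool.
# Depth testing is disabled so it draws over occluding geometry, and translucent vertex fog is turned off.
# Its graph holds a single VertexColor node, with no textures or parameters.
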Material /Game/Blueprints/Blutilities/ToolsWrkshp/ScrTools/TravelTime/Materials/M_TravelTime_TextMat
graph {
 properties {bDisableDepthTest=true, bUseTranslucencyVertexFog=false, blendMode=BLEND_Translucent, shadingModel=MSM_Unlit, usedShadingModels=Unlit}
 nodes [VertexColor()]
}
